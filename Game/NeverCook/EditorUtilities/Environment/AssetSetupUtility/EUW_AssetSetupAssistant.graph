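# Editor utility widget that creates new environment assets from templates, with type and parent-class dropdowns filled from a data table.
# Props duplicate the PropTemplate_A folder, rename its assets, then wire normals into spec composites, mesh into Blueprint defaults, material onto mesh and textures into parameters by name suffix.
# It optionally reparents the prop Blueprint to the selected class.
# Architecture instead duplicates blockout Blueprints into a category folder and renames them.
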EditorUtilityWidgetBlueprint /Game/NeverCook/EditorUtilities/Environment/AssetSetupUtility/EUW_AssetSetupAssistant
graphs [
 AssignMaterialToMesh {nodes [
   n1 = Cast To StaticMesh(Object:Wildcard)
   > n2 = Set L_MeshToModify(L_MeshToModify:Static Mesh Object Reference=@n1) -> (Output_Get:Static Mesh Object Reference)
   > RunAssetsThroughFilter!(
    execute
    self:Asset Registry Interface=GetAssetRegistry()
    AssetDataList:Array of Asset Data Structures=Get L_Assets() -> (L_Assets:Array of Asset Data Structures)
    Filter:"ARFilter Structure (by ref)"=MakeARFilter!(
     PackageNames:Array of Names
     PackagePaths:Array of Names
     SoftObjectPaths:Array of Soft Object Path Structures
     ClassPaths:Array of Top Level Asset Path Structures
     RecursiveClassPathsExclusionSet:Set of Top Level Asset Path Structures
     ClassNames:Array of Names=Make Array("[0]":Name=MaterialInstance) -> (Array:Array of Names)
     RecursiveClassesExclusionSet:Set of Names
     bRecursivePaths:bool=false
     bRecursiveClasses:bool=false
     bIncludeOnlyOnDiskAssets:bool=false
    ) -> (ReturnValue:ARFilter Structure)
   ) -> (then)
   > SetMaterial!(execute, self:Static Mesh Object Reference=@n2, MaterialIndex:int=0, NewMaterial:Material Interface Object Reference) -> (then)
   AssignMaterialToMesh() {is=title K2Node_FunctionEntry}
   > Set L_Assets(L_Assets:Array of Asset Data Structures=Get AR_RenamedAssets() -> (AR_RenamedAssets:Array of Asset Data Structures)) -> (Output_Get:Array of Asset Data Structures)
   > RunAssetsThroughFilter(
    self=GetAssetRegistry()
    AssetDataList=Get L_Assets()
    Filter=MakeARFilter(ClassNames=Make Array("[0]"=MaterialInstance), bRecursivePaths=false, bRecursiveClasses=false, bIncludeOnlyOnDiskAssets=false)
   )
   n3 = "Get (a copy)"(Array:Array of Object References=Get AR_NewMeshes() -> (AR_NewMeshes:Array of Object References)) -> (Output:Object Reference)
   > n4 = Cast To StaticMesh(Object:Object Reference=@n3)
   > n5 = Cast To MaterialInstanceConstant(Object="Get (a copy)"(Array=Get AR_NewMaterials() -> (AR_NewMaterials:Array of Object References)))
   > SetMaterial(self=@n4, MaterialIndex=0, NewMaterial=@n5)
   > SaveLoadedAsset(AssetToSave=@n3, bOnlyIfIsDirty=true)
   Set L_Assets(L_Assets=Get AR_RenamedAssets())
   > RunAssetsThroughFilter(
    self=GetAssetRegistry()
    AssetDataList=Get L_Assets()
    Filter=MakeARFilter(ClassNames=Make Array("[0]"=MaterialInstance), bRecursivePaths=false, bRecursiveClasses=false, bIncludeOnlyOnDiskAssets=false)
   )
   Break Asset Data(AssetData="Get (a copy)"(Array:Array of Asset Data Structures=Get L_Assets()) -> (Output:Asset Data Structure))
   "Get (a copy)"(Array:Array of Object References=Get AR_NewMaterials()) -> (Output:Object Reference)
   "Get (a copy)"(Array=Get AR_NewMaterials())
  ]}
 AssignMeshToBlueprint {nodes [
   n1 = Cast To BuildingActor!(execute, Object:Wildcard) -> (then, CastFailed, AsBuilding Actor:Building Actor Object Reference) {is=title K2Node_DynamicCast}
   n2 = "Get (a copy)"(Array=Get AR_NewMeshes())
   > n3 = Is Valid(InputObject=@n2)
   PrintString(execute=@n3.Is Not Valid, InString=Cast to BuildingActor Failed)
   n4 = SetEditorProperty(execute=@n3.Is Valid, Object=@n1, PropertyName=StaticMesh, PropertyValue:Object Reference=@n2)
   > n5 = Branch(Condition=@n4)
   > PrintString(InString=Mesh Update Successful)
   PrintString(execute=@n5.else, InString=Mesh Update Failed)
   AssignMeshToBlueprint() {is=title K2Node_FunctionEntry}
   > n6 = Set L_Assets(L_Assets=Get AR_RenamedAssets())
   > RunAssetsThroughFilter(
    self=GetAssetRegistry()
    AssetDataList=@n6
    Filter=MakeARFilter(ClassNames=Make Array("[0]"=Blueprint), bRecursivePaths=false, bRecursiveClasses=false, bIncludeOnlyOnDiskAssets=false)
   )
   n7 = "Get (a copy)"(Array:Array of Asset Data Structures=Get L_Assets()) -> (Output:Asset Data Structure)
   > Branch(Condition=Is Valid Asset Data!(InAssetData:"Asset Data Structure (by ref)"=@n7) -> (ReturnValue:bool))
   > n8 = Set L_ActorToUpdate(L_ActorToUpdate:Asset Data Structure=@n7) -> (Output_Get:Asset Data Structure)
   n9 = GetAsset(InAssetData=@n8)
   > n10 = GetBlueprintAsset(Object=@n9)
   > n11 = GeneratedClass!(execute, BlueprintObj:Blueprint Object Reference=@n10) -> (then, ReturnValue:Object Class Reference)
   > n12 = GetClassDefaultObject!(execute, Class:Object Class Reference=@n11) -> (then, ReturnValue:Object Reference)
   > SetEditorProperty(
    Object=@n12
    PropertyName=StaticMesh
    PropertyValue="Get (a copy)"(Array:Array of Object References=Get AR_NewMeshes()) -> (Output:Object Reference)
   )
   > SaveLoadedAsset(AssetToSave=@n9, bOnlyIfIsDirty=true)
  ]}
 AssignTexturesToMaterial {nodes [
   n1 = Cast To StaticMesh(Object="Get (a copy)"(Array=Get AR_NewMeshes()))
   > n2 = GetTextureParameterNames!(
    execute
    Material:Material Interface Object Reference=GetMaterial!(self:Static Mesh Object Reference=@n1, MaterialIndex:int=0) -> (ReturnValue:Material Interface Object Reference)
   ) -> (then, ParameterNames:Array of Names)
   > n3 = For Each Loop(Array:Array of Names=@n2) -> (Array Element:Name)
   > PrintString(InString=Append(A=TextureParameterName_, B="To String (Name)"(InName=@n3)))
   AssignTexturesToMaterial() {is=title K2Node_FunctionEntry}
   > n4 = Set L_Materials(L_Materials:Array of Asset Data Structures=Get AR_RenamedAssets()) -> (Output_Get:Array of Asset Data Structures)
   > RunAssetsThroughFilter(
    self=GetAssetRegistry()
    AssetDataList=@n4
    Filter=MakeARFilter(ClassNames=Make Array("[0]"=MaterialInstance), bRecursivePaths=false, bRecursiveClasses=false, bIncludeOnlyOnDiskAssets=false)
   )
   > n5 = Set L_Textures(L_Textures:Array of Asset Data Structures=Get AR_RenamedAssets()) -> (Output_Get:Array of Asset Data Structures)
   > RunAssetsThroughFilter(
    self=GetAssetRegistry()
    AssetDataList=@n5
    Filter=MakeARFilter(ClassNames=Make Array("[0]"=Texture2D), bRecursivePaths=false, bRecursiveClasses=false, bIncludeOnlyOnDiskAssets=false)
   )
   > n6 = For Each Loop(Array:Array of Object References=Get AR_NewMaterials()) -> (Array Element:Object Reference)
   > n7 = Cast To MaterialInstanceConstant(Object=@n6)
   > Set L_MaterialToModify(L_MaterialToModify:Material Instance Constant Object Reference=@n7) -> (Output_Get:Material Instance Constant Object Reference)
   > n8 = For Each Loop(Array=Get AR_NewTextures() -> (AR_NewTextures:Array of Object References))
   > n9 = Cast To Texture2D(Object=@n8)
   > Set L_TextureToAssign(L_TextureToAssign:Texture 2D Object Reference=@n9) -> (Output_Get:Texture 2D Object Reference)
   n10 = "To String (Object)"!(InObj:Object Reference=@n9) -> (ReturnValue:String)
   > n11 = Branch(Condition=MatchesWildcard(SourceString=Split(SourceString=@n10, InStr=_, SearchDir=FromEnd).RightS, Wildcard=D))
   > SetMaterialInstanceTextureParameterValue!(
    execute
    Instance:Material Instance Constant Object Reference=Get L_MaterialToModify() -> (L_MaterialToModify:Material Instance Constant Object Reference)
    ParameterName:Name=Diffuse
    Value:Texture Object Reference=Get L_TextureToAssign() -> (L_TextureToAssign:Texture 2D Object Reference)
    Association:EMaterialParameterAssociation Enum=GlobalParameter
   ) -> (then, ReturnValue:bool)
   > Set L_TextureToAssign()
   n12 = "To String (Object)"(InObj=Get L_TextureToAssign())
   n13 = Branch(execute=@n11.else, Condition=MatchesWildcard(SourceString=Split(SourceString=@n12, InStr=_, SearchDir=FromEnd).RightS, Wildcard=N))
   > SetMaterialInstanceTextureParameterValue(Instance=Get L_MaterialToModify(), ParameterName=Normals, Value=Get L_TextureToAssign(), Association=GlobalParameter)
   > Set L_TextureToAssign()
   n14 = "To String (Object)"(InObj=Get L_TextureToAssign())
   Branch(execute=@n13.else, Condition=MatchesWildcard(SourceString=Split(SourceString=@n14, InStr=_, SearchDir=FromEnd).RightS, Wildcard=S))
   > SetMaterialInstanceTextureParameterValue(Instance=Get L_MaterialToModify(), ParameterName=SpecularMasks, Value=Get L_TextureToAssign(), Association=GlobalParameter)
   > Set L_TextureToAssign()
   n15 = "To String (Object)"()
   n16 = Branch(Condition=MatchesWildcard(SourceString=Split(SourceString=@n15, InStr=_, SearchDir=FromEnd).RightS, Wildcard=S))
   PrintString(execute=@n16.else, InString=Spec not found)
   n17 = Branch(Condition=MatchesWildcard(SourceString=Split(SourceString=@n15, InStr=_, SearchDir=FromEnd).RightS, Wildcard=N))
   PrintString(execute=@n17.else, InString=Normals not found)
   n18 = Split(SourceString=@n15, InStr=_, SearchDir=FromEnd)
   n19 = Branch(Condition=MatchesWildcard(SourceString=@n18.RightS, Wildcard=D))
   PrintString(execute=@n19.else, InString=Diffuse not found)
   n20 = For Each Loop(Array:Array of Asset Data Structures) -> (Array Element:Asset Data Structure)
   Break Asset Data(AssetData=@n20)
   GetMaterialInstanceTextureParameterValue!(
    Instance:Material Instance Constant Object Reference
    ParameterName:Name="Get (a copy)"(Array:Array of Names=GetTextureParameterNames()) -> (Output:Name)
    Association:EMaterialParameterAssociation Enum=GlobalParameter
   ) -> (ReturnValue:Texture Object Reference)
   PrintString(InString=Append(A="ParameterName: "))
   SetTextureParameterValue(ParameterName=None)
   GetAsset(InAssetData=@n20)
   Append!(A=Split Result, B=": ", C:String=@n18.RightS) -> (ReturnValue)
   n21 = For Each Loop(Array:Array of Object References=Get AR_NewTextures()) -> (Array Element:Object Reference)
   GetEditorProperty!(execute:Exec, Object:Object Reference=@n21, PropertyName:Name=Diffuse) -> (then:Exec, PropertyValue:Wildcard, ReturnValue:bool)
   Get EditorAssetSubsystem() -> (ReturnValue:Editor Asset Subsystem Object Reference) {is=title K2Node_GetEditorSubsystem}
   n22 = "To String (Object)"()
   Branch(Condition=MatchesWildcard(SourceString=Split(SourceString=@n22, InStr=_, SearchDir=FromEnd).RightS, Wildcard=D))
   Branch()
   SetEditorProperty(Object=@n21, PropertyName=SpecularMasks, PropertyValue:"Wildcard (by ref)")
   SetEditorProperty(Object=@n21, PropertyName=Normals)
   SetEditorProperty(Object=@n21, PropertyName=Diffuse)
   Get AR_NewTextures()
  ]}
 CompositeMatchingNormal {nodes [
   n1 = CompositeMatchingNormal!() -> (then, OnlyCompositeMatchingNormalForTexturesEndingInThis:String=_S) {is=title K2Node_FunctionEntry}
   > Set l_ending(l_ending:String=@n1) -> (Output_Get:String)
   > n2 = Sequence()
   > n3 = Set l_Assets(l_Assets:Array of Asset Data Structures=Get AR_RenamedAssets()) -> (Output_Get:Array of Asset Data Structures)
   > RunAssetsThroughFilter(
    self=GetAssetRegistry()
    AssetDataList=@n3
    Filter=MakeARFilter(ClassNames=Make Array("[0]"=Texture2d), bRecursivePaths=false, bRecursiveClasses=false, bIncludeOnlyOnDiskAssets=false)
   )
   > n4 = For Each Loop(Array:Array of Asset Data Structures=Get l_Assets() -> (l_Assets:Array of Asset Data Structures)) -> (Array Element:Asset Data Structure)
   > n5 = Sequence()
   n6 = Break Asset Data!(AssetData=@n4) -> (AssetName, AssetClassPath)
   > n7 = Branch(Condition=EndsWith(SourceString="To String (Name)"(InName=@n6), InSuffix=_N))
   > Add!(
    execute
    TargetMap:Map of Names to Asset Data Structures=Get l_normalMapAssetData_Map() -> (l_normalMapAssetData_Map:Map of Names to Asset Data Structures)
    Key:"Name (by ref)"=@n6
    Value:"Asset Data Structure (by ref)"=@n4
   ) -> (then) {is=title K2Node_CallFunction}
   Branch(execute=@n7.else, Condition=EndsWith(SourceString="To String (Name)"(InName=@n6), InSuffix=Get l_ending() -> (l_ending:String)))
   > Add(TargetMap=Get l_specMapAssetData_Map() -> (l_specMapAssetData_Map:Map of Names to Asset Data Structures), Key=@n6, Value=@n4)
   n8 = Break Asset Data!(AssetData=@n4) -> (PackageName:Name, PackagePath:Name, AssetName, AssetClassPath)
   n9 = Branch(
    execute=@n5.then_1
    Condition=Contains Item!(TargetSet:Set of Names=Get l_pathsChecked() -> (l_pathsChecked:Set of Names), ItemToFind:"Name (by ref)"=@n8.PackagePath) -> (ReturnValue:bool) {is=title K2Node_CallFunction}
   )
   Add!(execute=@n9.else, TargetSet:Set of Names=Get l_pathsChecked(), NewItem:"Name (by ref)"=@n8.PackagePath) -> (then)
   > n10 = GetAssets!(
    execute
    self:Asset Registry Interface=GetAssetRegistry()
    Filter:"ARFilter Structure (by ref)"=MakeARFilter(
     PackagePaths=Make Array("[0]"=@n8.PackagePath)
     ClassNames=Make Array("[0]"=Texture2d)
     bRecursivePaths=false
     bRecursiveClasses=false
     bIncludeOnlyOnDiskAssets=false
    )
    bSkipARFilteredAssets:bool=true
   ) -> (then, OutAssetData:Array of Asset Data Structures, ReturnValue:bool)
   > n11 = For Each Loop(Array=@n10)
   n12 = Break Asset Data(AssetData=@n11)
   > Branch(Condition=EndsWith(SourceString="To String (Name)"(InName=@n12.AssetName), InSuffix=_N))
   > Add!(
    execute
    TargetMap:Map of Names to Asset Data Structures=Get l_normalMapAssetData_Map()
    Key:"Name (by ref)"=@n12.AssetName
    Value:"Asset Data Structure (by ref)"=@n11
   ) -> (then)
   > PrintString(InString=Append!(A="Adding normal: ", B="To String (Name)"(InName=@n12.AssetName)) -> (ReturnValue))
   n13 = Values!(execute=@n2.then_1, TargetMap:Map of Names to Asset Data Structures=Get l_specMapAssetData_Map()) -> (then, Values:Array of Asset Data Structures)
   > n14 = For Each Loop(Array=@n13)
   n15 = LeftChop(SourceString="To String (Name)"(InName=Break Asset Data(AssetData=@n14).AssetName), Count=Len(S=Get l_ending()))
   n16 = Find(
    TargetMap:Map of Names to Asset Data Structures=Get l_normalMapAssetData_Map()
    Key:"Name (by ref)"=String To Name(InString=Append(A=@n15, B=_N))
   ) -> (Value:Asset Data Structure)
   > Branch(Condition=@n16.ReturnValue)
   > n17 = Cast To Texture2D(Object=GetAsset(InAssetData=@n14))
   > n18 = Is Valid(InputObject=Get CompositeTexture!(self:Texture Object Reference=@n17) -> (CompositeTexture:Texture Object Reference))
   n19 = GetAsset(InAssetData=@n16)
   PrintString(
    execute=[@n18.Is Not Valid, @n18.Is Valid]
    InString=Append!(A=TextureBeingAssigned, B=_, C:String=GetDisplayName(Object=@n19)) -> (ReturnValue)
   )
   > n20 = Cast To Texture2D(Object=@n19)
   > SetEditorProperty(Object=@n17, PropertyName=CompositeTexture, PropertyValue:Texture 2D Object Reference=@n20)
   > SaveLoadedAsset(AssetToSave=@n17, bOnlyIfIsDirty=true)
   n21 = Set L_AR_AssetData(L_AR_AssetData:Array of Asset Data Structures=Get AR_NewAssetsInDirectory() -> (AR_NewAssetsInDirectory:Array of Asset Data Structures)) -> (Output_Get:Array of Asset Data Structures)
   > RunAssetsThroughFilter(
    self=GetAssetRegistry()
    AssetDataList=@n21
    Filter=MakeARFilter(ClassNames=Make Array("[0]"=Texture2d), bRecursivePaths=false, bRecursiveClasses=false, bIncludeOnlyOnDiskAssets=false)
   )
   > For Each Loop(Array:Array of Wildcards) -> (Array Element:Wildcard)
  ]}
 Debug_GetNumberOfNewAssets {nodes [
   Debug_GetNumberOfNewAssets() {is=title K2Node_FunctionEntry}
   n1 = Length(TargetArray:Array of Asset Data Structures=Get AR_NewAssetsInDirectory())
   > PrintString(InString=Append(A=Number of Assets Created, B=_, C="To String (Integer)"(InInt=@n1)))
  ]}
 Debug_GetSoftPathsOfNewAssets {nodes [
   Debug_GetSoftPathsOfNewAssets() {is=title K2Node_FunctionEntry}
   > PrintString(InString=Start:)
   > n1 = For Each Loop(Array:Array of Asset Data Structures=Get AR_NewAssetsInDirectory()) -> (Array Element:Asset Data Structure)
   PrintString(execute=@n1.Completed, InString=End)
   PrintString(
    execute=@n1.LoopBody
    InString=BreakSoftObjectPath(InSoftObjectPath=ToSoftObjectPath!(InAssetData:"Asset Data Structure (by ref)"=@n1) -> (ReturnValue:Soft Object Path Structure))
   )
  ]}
 Delete_SFX_Texture {nodes [
   Delete_SFX_Texture() {is=title K2Node_FunctionEntry}
   > n1 = Branch(Condition=Get Create_SFX_Mask() -> (Create_SFX_Mask:bool))
   n2 = For Each Loop(Exec=@n1.else, Array=Get AR_NewAssetsInDirectory())
   > n3 = Cast To Texture2D(Object=GetAsset(InAssetData=@n2))
   > Branch(Condition=EndsWith(SourceString="To String (Object)"(InObj=@n3), InSuffix=_M))
   > DeleteLoadedAsset!(execute, AssetToDelete:Object Reference=@n3) -> (then, ReturnValue:bool)
   PrintString(InString=Append!(A=MaskTexture_, B=GetDisplayName()) -> (ReturnValue))
   RemoveDuplicateSlashes!(InPath:String) -> (OutPath:String)
  ]}
 DuplicateTemplateAssets {nodes [
   DuplicateTemplateAssets() {is=title K2Node_FunctionEntry}
   > Clear(TargetArray:Array of Asset Data Structures=Get AR_NewAssetsInDirectory())
   n1 = Get ParentDirectoryPath() -> (ParentDirectoryPath:String)
   > DuplicateDirectory!(execute, SourceDirectoryPath:String=/Game/Environments/Outsource/Templates/PropTemplate_A/, DestinationDirectoryPath:String=@n1) -> (then, ReturnValue:bool)
   > n2 = DoesDirectoryHaveAssets!(execute, DirectoryPath:String=@n1, bRecursive:bool=true) -> (then, ReturnValue:bool)
   > n3 = Branch(Condition=@n2)
   > n4 = GetAssetsByPath!(
    execute
    self:Asset Registry Interface=GetAssetRegistry()
    PackagePath:Name=String To Name(InString=@n1)
    bRecursive:bool=true
    bIncludeOnlyOnDiskAssets:bool=false
   ) -> (then, OutAssetData:Array of Asset Data Structures, ReturnValue:bool)
   > Set AR_NewAssetsInDirectory(AR_NewAssetsInDirectory:Array of Asset Data Structures=@n4) -> (Output_Get:Array of Asset Data Structures)
   PrintString(execute=@n3.else, InString=No assets discovered)
   DuplicateAsset!(execute, SourceAssetPath:String, DestinationAssetPath:String) -> (then, ReturnValue:Object Reference)
  ]}
 EventGraph {nodes [
   n1 = "On Selection Changed (ddAssetType)"!() -> (OutputDelegate, then, SelectedItem:String, SelectionType:ESelectInfo Enum) {is=title K2Node_ComponentBoundEvent}
   > Set SelectedAssetType(SelectedAssetType:String=@n1.SelectedItem) -> (Output_Get:String)
   > HandleAssetTypeChange!(execute, self, SelectedItem:String=@n1.SelectedItem, SelectedType:ESelectInfo Enum=@n1.SelectionType) -> (then)
   > n2 = Branch(Condition=MatchesWildcard(SourceString=@n1.SelectedItem, Wildcard=Architecture))
   n3 = Get HB_ModKit_Options() -> (HB_ModKit_Options:Horizontal Box Object Reference)
   > SetVisibility(self=@n3, InVisibility=Visible)
   n4 = Get HB_PropParent() -> (HB_PropParent:Horizontal Box Object Reference)
   > n5 = For Each Loop(Array:Array of Horizontal Box Object References=Make Array!(
     "[0]":Horizontal Box Object Reference=@n4
     "[1]":Horizontal Box Object Reference=Get HB_PropName() -> (HB_PropName:Horizontal Box Object Reference)
    ) -> (Array:Array of Horizontal Box Object References)) -> (Array Element:Horizontal Box Object Reference)
   > n6 = SetVisibility(self=[@n5, @n4], InVisibility=Hidden)
   Set Text(
    execute=[@n6.then, @n7.then]
    Text="To Text (String)"(InString=PathGetFinalAssetPath!() -> (PathFinal:String))
    self:Text Box Object Reference=Get TB_AssetPath() -> (TB_AssetPath:Text Box Object Reference)
   )
   SetVisibility(execute=@n2.else, self=@n3, InVisibility=Hidden)
   > n7 = SetVisibility(self=[@n5, @n4], InVisibility=Visible)
   "On Clicked (BTN_CreateNewAsset)"!() -> (OutputDelegate, then, widget:User Widget Object Reference, event:Pointer Event Structure) {is=title K2Node_ComponentBoundEvent}
   n8 = Get SelectedAssetType() -> (SelectedAssetType:String)
   > n9 = Branch(Condition=MatchesWildcard(SourceString=@n8, Wildcard=Prop))
   > SetPropParentClass()
   > SetAssetPath()
   > DuplicateTemplateAssets() {is=title K2Node_CallFunction}
   > GetObjectRefFromAssetData()
   > RenameNewAssets()
   > FilterByAssetType()
   > CompositeMatchingNormal!(execute, self, OnlyCompositeMatchingNormalForTexturesEndingInThis:String=_S) -> (then) {is=title K2Node_CallFunction}
   > AssignMeshToBlueprint() {is=title K2Node_CallFunction}
   > AssignMaterialToMesh() {is=title K2Node_CallFunction}
   > AssignTexturesToMaterial() {is=title K2Node_CallFunction}
   > Branch(Condition=Get RunReparentFunction() -> (RunReparentFunction:bool))
   > ReparentBlueprint!(execute, self:Self Object Reference) -> (then)
   Branch(execute=@n9.else, Condition=MatchesWildcard(SourceString=@n8, Wildcard=Architecture))
   > Clear(TargetArray=Get AR_NewArchitectureAssets() -> (AR_NewArchitectureAssets:Array of Asset Data Structures))
   > n10 = For Each Loop(Array:Array of Strings=Get AR_SourceAssetPath() -> (AR_SourceAssetPath:Array of Strings)) -> (Array Element:String)
   n11 = Split(SourceString=@n10, InStr=., SearchDir=FromEnd)
   n12 = Get TXT_CategoryName() -> (TXT_CategoryName:Text)
   > DuplicateAsset(
    SourceAssetPath=@n10
    DestinationAssetPath=Append!(A=/Hermes/Sets/, B="To String (Text)"(InText=@n12), C:String=/, D:String=Blueprints/, E:String=@n11.RightS, F:String=., G:String=@n11.RightS) -> (ReturnValue)
   )
   n13 = "To String (Text)"(InText=Get TXT_CategoryName())
   > n14 = GetAssetsByPath(
    self=GetAssetRegistry()
    PackagePath=String To Name(InString=Append!(A=&Sets/, B=@n13, C=/, D=Blueprints/) -> (ReturnValue))
    bRecursive=true
    bIncludeOnlyOnDiskAssets=false
   )
   > n15 = Set AR_NewArchitectureAssets(AR_NewArchitectureAssets:Array of Asset Data Structures=@n14) -> (Output_Get:Array of Asset Data Structures)
   > n16 = For Each Loop(Array:Array of Asset Data Structures=@n15) -> (Array Element:Asset Data Structure)
   n17 = "To String (Name)"(InName=Break Asset Data(AssetData=@n16).AssetName)
   > PrintString(InString=Append!(A=DEBUG, B=_, C=@n17) -> (ReturnValue))
   n18 = Get TXT_CategoryName()
   > RenameAsset!(
    execute
    Asset:Object Reference=GetAsset(InAssetData=@n16)
    NewName:String=Replace(SourceString=@n17, From=Blockout, To="To String (Text)"(InText=@n18))
   ) -> (then)
   n19 = "On Text Committed (TB_Category_Name)"!() -> (OutputDelegate, then, Text:"Text (by ref)", CommitMethod:ETextCommit Enum=Default) {is=title K2Node_ComponentBoundEvent}
   > n20 = Set TXT_CategoryName(TXT_CategoryName:Text=@n19.Text) -> (Output_Get:Text)
   > Set Text(Text=@n20, self=Get TB_Category_Name() -> (TB_Category_Name:Text Box Object Reference))
   > Set Text(Text="To Text (String)"(InString=PathGetFinalAssetPath()), self=Get TB_AssetPath())
   "On Clicked (btnEditWidget)"!() -> (OutputDelegate, then, widget:User Widget Object Reference, event:Pointer Event Structure) {is=title K2Node_ComponentBoundEvent}
   > n21 = GetBlueprintAsset(Object=&self)
   > Execute Python Script!(
    execute
    PythonScript="import unreal\r\n\r\nmy_widget = widget_object\r\nasset_editor_subsystem = unreal.get_editor_subsystem(unreal.AssetEditorSubsystem)\r\nasset_editor_subsystem.open_editor_for_assets([my_widget])"
    widget_object:Blueprint Object Reference=@n21
   ) -> (then, ReturnValue) {inputs [widget_object]}
   n22 = "On Text Committed (TB_AssetName)"!() -> (OutputDelegate, then, Text:"Text (by ref)", CommitMethod:ETextCommit Enum=Default) {is=title K2Node_ComponentBoundEvent}
   > n23 = Set TXT_AssetName(TXT_AssetName:Text=@n22.Text) -> (Output_Get:Text)
   > Set Text(Text=@n23, self=Get TB_AssetName() -> (TB_AssetName:Text Box Object Reference))
   > Set Text(Text="To Text (String)"(InString=PathGetFinalAssetPath()), self=Get TB_AssetPath())
   Event PreConstruct()
   > n24 = Sequence!(execute) -> (then_0, then_1, then_2:Exec)
   n25 = Self-Reference()
   n26 = Get DV_PathAssets() -> (DV_PathAssets:Details View Object Reference)
   > SetObject(self=@n26, NewObject=@n25)
   > SetObject(self=@n26, NewObject=@n25)
   n27 = Self-Reference()
   n28 = Get DV_CustomOptions() -> (DV_CustomOptions:Details View Object Reference)
   SetObject(execute=@n24.then_1, self=@n28, NewObject=@n27)
   > SetObject(self=@n28, NewObject=@n27)
   Event Construct()
   > Init()
   > n29 = Branch(Condition="Equal Exactly (String)"(A=GetPlatformUserName!() -> (ReturnValue:String), B=brianfasten))
   n30 = Get btnEditWidget() -> (btnEditWidget:TA Generic 01 Button Object Reference)
   > SetVisibility(self=@n30, InVisibility=Visible)
   SetVisibility(execute=@n29.else, self=@n30, InVisibility=Hidden)
   Branch(Condition=MatchesWildcard(Wildcard=Core))
   > n31 = For Each Loop(Array:Array of Object References=Get AR_CoreKit_Classes() -> (AR_CoreKit_Classes:Array of Object References)) -> (Array Element:Object Reference)
   > DuplicateAsset!(
    execute:Exec
    self:Object Reference=Self-Reference()
    AssetName:String="To String (Text)"(InText=Get TXT_AssetName() -> (TXT_AssetName:Text))
    PackagePath:String=PathGetFinalAssetPath()
    OriginalObject:Object Reference=@n31
   ) -> (then:Exec, ReturnValue:Object Reference) {is=title K2Node_Message}
   "To String (Name)"(InName=Break Asset Data!(AssetData=CreateAssetData(InAsset=@n31, bAllowBlueprintClass=false)) -> (PackageName, PackagePath, AssetName, AssetClassPath, AssetClass:Name).AssetClass)
   PathGetFinalAssetPath()
   DuplicateDirectory(
    SourceDirectoryPath=&PropTemplate_A/
    DestinationDirectoryPath=/Game/Environments/Outsource/Templates/DevTest/
   )
   DuplicateAsset(
    SourceAssetPath=/Game/Environments/Outsource/Templates/PropTemplate_A/Textures/T_Prop_A_D
    DestinationAssetPath=/Game/Environments/Outsource/Templates/PropTemplate_A/Textures/T_Prop_B_D
   ) {is=title K2Node_CallFunction}
   DuplicateAsset(
    SourceAssetPath=&T_Prop_A_D
    DestinationAssetPath=/Game/Environments/Outsource/Templates/PropTemplate_A/Textures/T_EansGenericTexture_A_D
   )
   DuplicateDirectory(SourceDirectoryPath=&PropTemplate_A/)
   PrintString(InString=Hello)
   Delete_SFX_Texture() {is=title K2Node_CallFunction}
   n32 = ToSoftObjectPath()
   PrintString(InString=Append(A=Soft Object Path:, B=" -  ", C=BreakSoftObjectPath(InSoftObjectPath=@n32)))
   PrintString(InString=I'm ready)
   DuplicateAsset()
   PrintString(InString=Hello)
   n33 = Length(TargetArray=@n15)
   PrintString(InString=Append!(A=DEBUG_, B="To String (Integer)"(InInt=@n33)) -> (ReturnValue))
   Append!(A=/Game/Environments/Blockout/Blueprints/Walls/BP_Blockout_Wall_A_A, B=BP_Blockout_Wall_A_A, C:String=/) -> (ReturnValue)
   Append!(
    A=/Hermes/
    B=Sets/
    C="To String (Text)"(InText=Get TXT_CategoryName())
    D:String=/
    E:String=Blueprints
    F:String=/
    G:String=BP_Blockout_Wall_A_A
    H:String=.
    I:String=BP_Blockout_Wall_A_A
   ) -> (ReturnValue)
   "On Selection Changed (ddKitOptions)"!() -> (OutputDelegate, then, SelectedItem:String, SelectionType:ESelectInfo Enum) {is=title K2Node_ComponentBoundEvent}
   Event Tick()
  ]}
 FilterByAssetType {nodes [
   FilterByAssetType() {is=title K2Node_FunctionEntry}
   > n1 = Sequence()
   > n2 = Filter by Class!(
    execute
    TargetArray:Array of Object References=Get AR_ObjectRefs() -> (AR_ObjectRefs:Array of Object References)
    ObjectClass:Object Class Reference=Get L_OC_MatInstance() -> (L_OC_MatInstance:Object Class Reference)
    FilterType:EEditorScriptingFilterType Enum
   ) -> (then, ReturnValue:Array of Object References)
   > Set AR_NewMaterials(AR_NewMaterials:Array of Object References=@n2) -> (Output_Get:Array of Object References)
   n3 = Filter by Class(execute=@n1.then_1, TargetArray=Get AR_ObjectRefs(), ObjectClass=Get L_OC_Texture() -> (L_OC_Texture:Object Class Reference))
   > Set AR_NewTextures(AR_NewTextures:Array of Object References=@n3) -> (Output_Get:Array of Object References)
   n4 = Filter by Class(execute=@n1.then_2, TargetArray=Get AR_ObjectRefs(), ObjectClass=Get L_OC_StaticMesh() -> (L_OC_StaticMesh:Object Class Reference))
   > Set AR_NewMeshes(AR_NewMeshes:Array of Object References=@n4) -> (Output_Get:Array of Object References)
  ]}
 GetObjectRefFromAssetData {nodes [
   GetObjectRefFromAssetData() {is=title K2Node_FunctionEntry}
   > Clear(TargetArray:Array of Object References=Get AR_ObjectRefs())
   > n1 = For Each Loop(Array:Array of Asset Data Structures=Get AR_NewAssetsInDirectory()) -> (Array Element:Asset Data Structure)
   > Add Unique(TargetArray:Array of Object References=Get AR_ObjectRefs(), NewItem:"Object Reference (by ref)"=GetAsset(InAssetData=@n1))
  ]}
 GetTemplateAsset {nodes [
   GetSelectedIndex!(self:"Searchable Combo Box (String) Object Reference"=Get ddParentClass() -> (ddParentClass:"Searchable Combo Box (String) Object Reference")) -> (ReturnValue:int=0)
   GetTemplateAsset() {is=title K2Node_FunctionEntry}
  ]}
 HandleAssetTypeChange {nodes [
   n1 = HandleAssetTypeChange!() -> (then, SelectedItem:String, SelectedType:ESelectInfo Enum) {is=title K2Node_FunctionEntry}
   > ClearOptions(self:"Searchable Combo Box (String) Object Reference"=Get ddParentClass())
   > n2 = Get Data Table Row!(
    execute:Exec
    DataTable:Data Table Object Reference=Get AssetTypeDataTable() -> (AssetTypeDataTable:Data Table Object Reference)
    RowName:Name=String To Name(InString=@n1)
   ) -> (then:Exec, RowNotFound:Exec, ReturnValue:STR ASU Asset Type Structure)
   n3 = Break STR ASU Asset Type!(STR_ASU_AssetType:"STR ASU Asset Type Structure (by ref)"=@n2) -> (
    PathName_3_#1:String
    Classes_13_#2:Map of Object Soft Class References to Strings
   ) {is=title K2Node_BreakStruct}
   > n4 = Values(TargetMap:Map of Object Soft Class References to Strings=@n3.Classes_13_#2) -> (Values:Array of Strings)
   > n5 = Keys!(execute, TargetMap:Map of Object Soft Class References to Strings=@n3.Classes_13_#2) -> (then, Keys:Array of Object Soft Class References)
   > n6 = For Each Loop(Array:Array of Object Soft Class References=@n5) -> (Array Element:Object Soft Class Reference)
   n7 = Get ddParentClass()
   SetIsEnabled(
    execute=@n6.Completed
    self=@n7
    bInIsEnabled="integer > integer"(A=GetOptionCount!(self:"Searchable Combo Box (String) Object Reference"=@n7) -> (ReturnValue:int=0), B=1)
   )
   > SetSelectedIndex!(execute, self:"Searchable Combo Box (String) Object Reference"=@n7, Index:int=1) -> (then)
   AddOption!(
    execute=@n6.LoopBody
    self:"Searchable Combo Box (String) Object Reference"=Get ddParentClass()
    Option:String="Get (a ref)"(Array:Array of Strings=@n4, Dimension 1=@n6.Array Index) -> (Output:"String (by ref)")
   ) -> (then)
  ]}
 Init {nodes [
   Init() {is=title K2Node_FunctionEntry}
   > Is Valid(InputObject=Get AssetTypeDataTable())
   > Sequence!(execute) -> (then_0, then_1)
   > n1 = GetDataTableRowNames(Table=Get AssetTypeDataTable())
   n2 = Get ddAssetType() -> (ddAssetType:"Searchable Combo Box (String) Object Reference")
   > ClearOptions(self=@n2)
   > n3 = For Each Loop(Array:Array of Names=@n1) -> (Array Element:Name)
   SetSelectedIndex(execute=@n3.Completed, self=Get ddAssetType(), Index=1)
   AddOption(execute=@n3.LoopBody, self=@n2, Option="To String (Name)"(InName=@n3))
  ]}
 PathGetFinalAssetPath {nodes [
   PathGetFinalAssetPath() {is=title K2Node_FunctionEntry}
   n1 = "To String (Text)"(InText=Get TXT_CategoryName())
   n2 = PathGetType!() -> (PathType:String)
   n3 = "To String (Text)"(InText=Get TXT_AssetName())
   > Return Node!(
    execute
    PathFinal:String=Append!(A=PathGetStart!() -> (PathStart:String), B=@n2, C=/, D=@n1, E=/, F=Append!(A=@n1, B=_, C=@n3) -> (ReturnValue), G=/) -> (ReturnValue)
   )
   LeftChop(SourceString=@n2, Count=1)
  ]}
 PathGetStart {nodes [
   PathGetStart() {is=title K2Node_FunctionEntry}
   > n1 = Branch(Condition=Get UsePlugin() -> (UsePlugin:bool))
   > n2 = Set L_PathStart(L_PathStart:String=/Plugins/Hermes/) -> (Output_Get:String)
   Return Node!(execute=[@n3.then, @n2.then], PathStart:String=Get L_PathStart() -> (L_PathStart:String))
   n3 = Set UsePlugin(execute=@n1.else, UsePlugin:bool=true) -> (Output_Get:bool)
  ]}
 PathGetType {nodes [
   PathGetType() {is=title K2Node_FunctionEntry}
   n1 = Get ddAssetType()
   > n2 = Get Data Table Row(
    DataTable=Get AssetTypeDataTable()
    RowName=String To Name(InString=GetSelectedOption!(self:"Searchable Combo Box (String) Object Reference"=@n1) -> (ReturnValue:String))
   )
   n3 = Break STR ASU Asset Type(STR_ASU_AssetType=@n2)
   > n4 = Branch(Condition="Equal (Name)"(A=String To Name(InString=@n3), B=Props))
   > Return Node!(execute, PathType:String=@n3)
   Return Node(execute=@n4.else, PathType=Get L_AssetType() -> (L_AssetType:String))
  ]}
 RenameNewAssets {nodes [
   RenameNewAssets() {is=title K2Node_FunctionEntry}
   > Clear(TargetArray:Array of Asset Data Structures=Get AR_RenamedAssets())
   > n1 = For Each Loop(Array:Array of Asset Data Structures=Get AR_NewAssetsInDirectory()) -> (Array Element:Asset Data Structure)
   n2 = GetAssetsByPath(
    execute=@n1.Completed
    self=GetAssetRegistry()
    PackagePath=String To Name(InString=Get ParentDirectoryPath())
    bRecursive=true
    bIncludeOnlyOnDiskAssets=false
   )
   > Set AR_RenamedAssets(AR_RenamedAssets:Array of Asset Data Structures=@n2) -> (Output_Get:Array of Asset Data Structures)
   n3 = Break Asset Data!(AssetData=@n1) -> (PackageName, PackagePath, AssetName, AssetClassPath)
   RenameAsset(
    execute=@n1.LoopBody
    Asset=GetAsset(InAssetData=@n1)
    NewName=Replace(SourceString="To String (Name)"(InName=@n3.AssetName), From=Prop, To=Get UpdatedAssetName() -> (UpdatedAssetName:String))
   )
  ]}
 ReparentBlueprint {nodes [
   n1 = Cast To BuildingActor()
   n2 = "Get (a copy)"(Array:Array of Object References=Get AR_NewMeshes()) -> (Output:Object Reference)
   > n3 = Is Valid(InputObject=@n2)
   PrintString(execute=@n3.Is Not Valid, InString=Cast to BuildingActor Failed)
   n4 = SetEditorProperty(execute=@n3.Is Valid, Object=@n1, PropertyName=StaticMesh, PropertyValue:Object Reference=@n2)
   > n5 = Branch(Condition=@n4)
   > PrintString(InString=Mesh Update Successful)
   PrintString(execute=@n5.else, InString=Mesh Update Failed)
   ReparentBlueprint() {is=title K2Node_FunctionEntry}
   > Set L_NewParentClass(L_NewParentClass:Object Class Reference=Get SelectedParentClass() -> (SelectedParentClass:Object Class Reference)) -> (Output_Get:Object Class Reference)
   > n6 = Set L_Assets(L_Assets=Get AR_RenamedAssets())
   > RunAssetsThroughFilter(
    self=GetAssetRegistry()
    AssetDataList=@n6
    Filter=MakeARFilter(
     ClassNames=Make Array!("[0]":Name=Blueprint) -> (Array:Array of Names)
     bRecursivePaths=false
     bRecursiveClasses=false
     bIncludeOnlyOnDiskAssets=false
    )
   )
   n7 = "Get (a copy)"(Array:Array of Asset Data Structures=Get L_Assets()) -> (Output:Asset Data Structure)
   > Branch(Condition=Is Valid Asset Data(InAssetData=@n7))
   > n8 = Sequence()
   > n9 = Set L_ActorToUpdate(L_ActorToUpdate=@n7)
   n10 = GetAsset(InAssetData=@n9)
   > n11 = GetBlueprintAsset(Object=@n10)
   > ReparentBlueprint!(
    execute
    Blueprint:Blueprint Object Reference=@n11
    NewParentClass:Object Class Reference=Get L_NewParentClass() -> (L_NewParentClass:Object Class Reference)
   ) -> (then) {is=title K2Node_CallFunction}
   SaveLoadedAsset(execute=@n8.then_1, AssetToSave=@n10, bOnlyIfIsDirty=true)
  ]}
 SetArchitecturePath {nodes [
   SetArchitecturePath() {is=title K2Node_FunctionEntry}
   > Set L_EnvCategory(L_EnvCategory:String="To String (Text)"(InText=Get TXT_CategoryName())) -> (Output_Get:String)
   > Set L_AssetName(L_AssetName:String="To String (Text)"(InText=Get TXT_AssetName())) -> (Output_Get:String)
   n1 = Append!(
    A=Append!(A=Get L_RootPath() -> (L_RootPath:String), B=/) -> (ReturnValue)
    B=Append(A=Get L_EnvClass() -> (L_EnvClass:String), B=/)
    C=Append(A=Get L_EnvCategory() -> (L_EnvCategory:String), B=/)
    D:String
   ) -> (ReturnValue)
   > PrintString(InString=@n1)
   > Set ParentDirectoryPath(ParentDirectoryPath:String=@n1) -> (Output_Get:String)
   n2 = Get L_AssetName() -> (L_AssetName:String)
   > Set UpdatedAssetName(UpdatedAssetName:String=Append!(A=Get L_EnvCategory(), B=_, C:String=@n2) -> (ReturnValue)) -> (Output_Get:String)
   DuplicateDirectory(
    SourceDirectoryPath=&PropTemplate_A/
    DestinationDirectoryPath=&DevTest/
   )
   PathGetFinalAssetPath() {is=title K2Node_CallFunction}
   LeftChop(SourceString=@n2, Count=2)
  ]}
 SetAssetPath {nodes [
   SetAssetPath() {is=title K2Node_FunctionEntry}
   > Set L_EnvCategory(L_EnvCategory="To String (Text)"(InText=Get TXT_CategoryName()))
   > Set L_AssetName(L_AssetName="To String (Text)"(InText=Get TXT_AssetName()))
   n1 = Get L_EnvCategory()
   n2 = Get L_AssetName()
   n3 = Append!(
    A=Append!(A=Get L_RootPath(), B=/) -> (ReturnValue)
    B=Append(A=Get L_EnvClass(), B=/)
    C=Append(A=@n1, B=/)
    D:String=Append(A=Append!(A=@n1, B=_, C:String=@n2) -> (ReturnValue), B=/)
   ) -> (ReturnValue)
   > PrintString(InString=@n3)
   > Set ParentDirectoryPath(ParentDirectoryPath=@n3)
   n4 = Get L_AssetName()
   > Set UpdatedAssetName(UpdatedAssetName=Append(A=Get L_EnvCategory(), B=_, C=@n4))
   DuplicateDirectory(
    SourceDirectoryPath=&PropTemplate_A/
    DestinationDirectoryPath=&DevTest/
   )
   PathGetFinalAssetPath()
   LeftChop(SourceString=@n4, Count=2)
  ]}
 SetPropParentClass {nodes [
   SetPropParentClass() {is=title K2Node_FunctionEntry}
   n1 = GetSelectedIndex(self=Get ddParentClass()) -> (ReturnValue=0)
   > n2 = Branch(Condition="Equal (Integer)"(A=@n1))
   > Set SelectedParentClass(SelectedParentClass:Object Class Reference=/Game/Environments/Blockout/Blueprints/Props/BP_Blockout_Prop_Small_A.BP_Blockout_Prop_Small_A_C) -> (Output_Get:Object Class Reference)
   n3 = Branch(execute=@n2.else, Condition="Equal (Integer)"(A=@n1, B=1))
   > PrintString(InString=Default prop class is medium)
   > Set RunReparentFunction(RunReparentFunction:bool=false) -> (Output_Get:bool)
   n4 = Branch(execute=@n3.else, Condition="Equal (Integer)"(A=@n1, B=2))
   > Set SelectedParentClass(SelectedParentClass=/Game/Environments/Blockout/Blueprints/Props/BP_Blockout_Prop_Large_A.BP_Blockout_Prop_Large_A_C)
   Branch(execute=@n4.else, Condition="Equal (Integer)"(A=@n1, B=3))
   > Set SelectedParentClass(SelectedParentClass=/Game/Environments/Apollo/Blueprints/RandomColor_Prop/RandomColor_Prop.RandomColor_Prop_C)
  ]}
]
variables [
 AR_CoreKit_Classes {
  type="array<Object>"
  default [
   &BP_Blockout_Wall_A_A
   /Game/Environments/Blockout/Blueprints/Walls/Doorway/BP_Blockout_Doorway_A_A_A
  ]
 }
 AR_NewArchitectureAssets {type="array<AssetData>", default []}
 AR_NewAssetsInDirectory {type="array<AssetData>", default []}
 AR_NewMaterials {type="array<Object>", default []}
 AR_NewMeshes {type="array<Object>", default []}
 AR_NewTextures {type="array<Object>", default []}
 AR_ObjectRefs {type="array<Object>", default []}
 AR_RenamedAssets {type="array<AssetData>", default []}
 AR_SourceAssetPath {
  type="array<string>"
  default [
   &BP_Blockout_Wall_A_A
   &BP_Blockout_Doorway_A_A_A
   /Game/Environments/Blockout/Blueprints/Walls/Window/Window_A/BP_Blockout_Diag_Window_A_A_B
   /Game/Environments/Blockout/Blueprints/Walls/Window/Window_A/BP_Blockout_Diag_Window_A_B_B
   /Game/Environments/Blockout/Blueprints/Roof/BP_Blockout_Roof_A_A
  ]
 }
 AssetTypeDataTable {type=DataTable, default=/Game/NeverCook/EditorUtilities/Environment/AssetSetupUtility/Data/DT_ASU_AssetTypes}
 Create_SFX_Mask {type=boolean, default=true}
 ParentDirectoryPath {type=string, default=""}
 RunReparentFunction {type=boolean, default=true}
 Selected Item {type=string, default=""}
 SelectedAssetType {type=string, default=""}
 SelectedParentClass {type=Class, default=null}
 Selection Type {type=ESelectInfo, default=OnMouseClick}
 SM_TemplateMesh {type=StaticMesh, default=/Game/Environments/Outsource/Templates/PropTemplate_A/Meshes/SM_Prop_A}
 TXT_AssetName {type=string, default=""}
 TXT_CategoryName {type=string, default=""}
 UpdatedAssetName {type=string, default=""}
 UsePlugin {type=boolean, default=true}
]
widgets {
 info {rootWidgetClass=&VerticalBox, widgetCount=27}
 widgets [
  VerticalBox_0 {widgetClassPath=&VerticalBox, widget=VerticalBox_0}
  HorizontalBox_3 {
   widgetClassPath=&HorizontalBox
   parent=VerticalBox_0
   widget=HorizontalBox_3
   slot {class=VerticalBoxSlot, properties {horizontalAlignment=HAlign_Center, padding=<ltrb 5 5 5 5>}}
  }
  TextBlock_36 {
   widgetClassPath=&TextBlock
   parent=HorizontalBox_3
   widget=TextBlock_36
   properties {justification=Center, text=Environment Asset Setup Assistant}
   slot {class=HorizontalBoxSlot}
  }
  btnEditWidget {
   widgetClassPath=&TA_Generic01_Button
   parent=HorizontalBox_3
   widget=btnEditWidget
   properties {label=">>", buttonColor_Default=<rgba 0.0469 0.0469 0.0469 1>, sizeOverride=<20 0>}
   slot {class=HorizontalBoxSlot, properties {padding=<ltrb 10 10 10 10>}}
  }
  DV_PathAssets {
   widgetClassPath=&DetailsView
   parent=VerticalBox_0
   widget=DV_PathAssets
   properties {visibility=Hidden, categoriesToShow [PathAssets]}
   slot {class=VerticalBoxSlot}
  }
  VerticalBox_1 {
   widgetClassPath=&VerticalBox
   parent=VerticalBox_0
   widget=VerticalBox_1
   slot {class=VerticalBoxSlot, properties {padding=<ltrb 10 10 10 10>}}
  }
  HorizontalBox_0 {widgetClassPath=&HorizontalBox, parent=VerticalBox_1, widget=HorizontalBox_0, slot {class=VerticalBoxSlot}}
  CommonTextBlock_89 {
   widgetClassPath=&CommonTextBlock
   parent=HorizontalBox_0
   widget=CommonTextBlock_89
   properties {
    style=&TextStyle-Base_C
    text=Asset Type:
    font {
     bForceMonospaced=false
     bMaterialIsStencil=false
     fontMaterial=null
     fontObject=&NotoSans
     letterSpacing=0
     monospacedWidth=1
     size=24
     skewAmount=0
     typefaceFontName=NotoSans-Regular
     outlineSettings {
      bApplyOutlineToDropShadows=false
      bMiteredCorners=false
      bSeparateFillAlpha=false
      outlineMaterial=null
      outlineSize=0
      outlineColor=<rgba 0 0 0 1>
     }
    }
    shadowOffset=<0 0>
   }
   slot {class=HorizontalBoxSlot}
  }
  ddAssetType {
   widgetClassPath=/Script/FortniteEditor.SearchableStringComboBox
   bIsVariable=true
   parent=HorizontalBox_0
   widget=ddAssetType
   properties {
    selectedOption=Architecture
    defaultOptions [Architecture, Prop, Indestructible]
    foregroundColor {colorUseRule=UseColor_Specified, specifiedColor=<rgba 0.5271 0.5271 0.5271 1>}
    itemStyle {
     bUseParentRowBrush=false
     activeBrush {
      bIsDynamicallyLoaded=false
      drawAs=Image
      imageType=NoImage
      mirroring=NoMirror
      resourceName=None
      resourceObject=null
      tiling=NoTile
      imageSize=<0 0>
      margin=<ltrb 0 0 0 0>
      outlineSettings {
       bUseBrushTransparency=false
       roundingType=HalfHeightRadius
       width=0
       color {colorUseRule=UseColor_Specified, specifiedColor=<rgba 0 0 0 0>}
       cornerRadii=<0 0 0 0>
      }
      tintColor {colorUseRule=UseColor_Specified, specifiedColor=<rgba 0 0.162 0.7454 1>}
      uVRegion {bIsValid=false, max=<0 0>, min=<0 0>}
     }
     activeHighlightedBrush {
      bIsDynamicallyLoaded=false
      drawAs=Image
      imageType=NoImage
      mirroring=NoMirror
      resourceName=None
      resourceObject=null
      tiling=NoTile
      imageSize=<0 0>
      margin=<ltrb 0 0 0 0>
      outlineSettings {
       bUseBrushTransparency=false
       roundingType=HalfHeightRadius
       width=0
       color {colorUseRule=UseColor_Specified, specifiedColor=<rgba 0 0 0 0>}
       cornerRadii=<0 0 0 0>
      }
      tintColor {colorUseRule=UseColor_Specified, specifiedColor=<rgba 0.0252 0.0319 0.0423 1>}
      uVRegion {bIsValid=false, max=<0 0>, min=<0 0>}
     }
     activeHoveredBrush {
      bIsDynamicallyLoaded=false
      drawAs=Image
      imageType=NoImage
      mirroring=NoMirror
      resourceName=None
      resourceObject=null
      tiling=NoTile
      imageSize=<0 0>
      margin=<ltrb 0 0 0 0>
      outlineSettings {
       bUseBrushTransparency=false
       roundingType=HalfHeightRadius
       width=0
       color {colorUseRule=UseColor_Specified, specifiedColor=<rgba 0 0 0 0>}
       cornerRadii=<0 0 0 0>
      }
      tintColor {colorUseRule=UseColor_Specified, specifiedColor=<rgba 0 0.162 0.7454 1>}
      uVRegion {bIsValid=false, max=<0 0>, min=<0 0>}
     }
     dropIndicator_Above {
      bIsDynamicallyLoaded=false
      drawAs=Box
      imageType=FullColor
      mirroring=NoMirror
      resourceName=../../../Engine/Content/Slate/Common/DropZoneIndicator_Above.png
      resourceObject=null
      tiling=NoTile
      imageSize=<0 0>
      margin=<ltrb 0.625 0.625 0 0>
      outlineSettings {
       bUseBrushTransparency=false
       roundingType=HalfHeightRadius
       width=0
       color {colorUseRule=UseColor_Specified, specifiedColor=<rgba 0 0 0 0>}
       cornerRadii=<0 0 0 0>
      }
      tintColor {colorUseRule=UseColor_Specified, specifiedColor=<rgba 0.0091 0.3515 0.9647 1>}
      uVRegion {bIsValid=false, max=<0 0>, min=<0 0>}
     }
     dropIndicator_Below {
      bIsDynamicallyLoaded=false
      drawAs=Box
      imageType=FullColor
      mirroring=NoMirror
      resourceName=../../../Engine/Content/Slate/Common/DropZoneIndicator_Below.png
      resourceObject=null
      tiling=NoTile
      imageSize=<0 0>
      margin=<ltrb 0.625 0 0 0.625>
      outlineSettings {
       bUseBrushTransparency=false
       roundingType=HalfHeightRadius
       width=0
       color {colorUseRule=UseColor_Specified, specifiedColor=<rgba 0 0 0 0>}
       cornerRadii=<0 0 0 0>
      }
      tintColor {colorUseRule=UseColor_Specified, specifiedColor=<rgba 0.0091 0.3515 0.9647 1>}
      uVRegion {bIsValid=false, max=<0 0>, min=<0 0>}
     }
     dropIndicator_Onto {
      bIsDynamicallyLoaded=false
      drawAs=Box
      imageType=FullColor
      mirroring=NoMirror
      resourceName=../../../Engine/Content/Slate/Common/DropZoneIndicator_Onto.png
      resourceObject=null
      tiling=NoTile
      imageSize=<0 0>
      margin=<ltrb 0.25 0.25 0.25 0.25>
      outlineSettings {
       bUseBrushTransparency=false
       roundingType=HalfHeightRadius
       width=0
       color {colorUseRule=UseColor_Specified, specifiedColor=<rgba 0 0 0 0>}
       cornerRadii=<0 0 0 0>
      }
      tintColor {colorUseRule=UseColor_Specified, specifiedColor=<rgba 0.0091 0.3515 0.9647 1>}
      uVRegion {bIsValid=false, max=<0 0>, min=<0 0>}
     }
     evenRowBackgroundBrush {
      bIsDynamicallyLoaded=false
      drawAs=Image
      imageType=NoImage
      mirroring=NoMirror
      resourceName=None
      resourceObject=null
      tiling=NoTile
      imageSize=<0 0>
      margin=<ltrb 0 0 0 0>
      outlineSettings {
       bUseBrushTransparency=false
       roundingType=HalfHeightRadius
       width=0
       color {colorUseRule=UseColor_Specified, specifiedColor=<rgba 0 0 0 0>}
       cornerRadii=<0 0 0 0>
      }
      tintColor {colorUseRule=UseColor_Specified, specifiedColor=<rgba 0.0103 0.0103 0.0103 1>}
      uVRegion {bIsValid=false, max=<0 0>, min=<0 0>}
     }
     evenRowBackgroundHoveredBrush {
      bIsDynamicallyLoaded=false
      drawAs=Image
      imageType=NoImage
      mirroring=NoMirror
      resourceName=None
      resourceObject=null
      tiling=NoTile
      imageSize=<0 0>
      margin=<ltrb 0 0 0 0>
      outlineSettings {
       bUseBrushTransparency=false
       roundingType=HalfHeightRadius
       width=0
       color {colorUseRule=UseColor_Specified, specifiedColor=<rgba 0 0 0 0>}
       cornerRadii=<0 0 0 0>
      }
      tintColor {colorUseRule=UseColor_Specified, specifiedColor=<rgba 0.0176 0.0176 0.0176 1>}
      uVRegion {bIsValid=false, max=<0 0>, min=<0 0>}
     }
     inactiveBrush {
      bIsDynamicallyLoaded=false
      drawAs=Image
      imageType=NoImage
      mirroring=NoMirror
      resourceName=None
      resourceObject=null
      tiling=NoTile
      imageSize=<0 0>
      margin=<ltrb 0 0 0 0>
      outlineSettings {
       bUseBrushTransparency=false
       roundingType=HalfHeightRadius
       width=0
       color {colorUseRule=UseColor_Specified, specifiedColor=<rgba 0 0 0 0>}
       cornerRadii=<0 0 0 0>
      }
      tintColor {colorUseRule=UseColor_Specified, specifiedColor=<rgba 0.0513 0.0953 0.159 1>}
      uVRegion {bIsValid=false, max=<0 0>, min=<0 0>}
     }
     inactiveHighlightedBrush {
      bIsDynamicallyLoaded=false
      drawAs=Image
      imageType=NoImage
      mirroring=NoMirror
      resourceName=None
      resourceObject=null
      tiling=NoTile
      imageSize=<0 0>
      margin=<ltrb 0 0 0 0>
      outlineSettings {
       bUseBrushTransparency=false
       roundingType=HalfHeightRadius
       width=0
       color {colorUseRule=UseColor_Specified, specifiedColor=<rgba 0 0 0 0>}
       cornerRadii=<0 0 0 0>
      }
      tintColor {colorUseRule=UseColor_Specified, specifiedColor=<rgba 0.0252 0.0319 0.0423 1>}
      uVRegion {bIsValid=false, max=<0 0>, min=<0 0>}
     }
     inactiveHoveredBrush {
      bIsDynamicallyLoaded=false
      drawAs=Image
      imageType=NoImage
      mirroring=NoMirror
      resourceName=None
      resourceObject=null
      tiling=NoTile
      imageSize=<0 0>
      margin=<ltrb 0 0 0 0>
      outlineSettings {
       bUseBrushTransparency=false
       roundingType=HalfHeightRadius
       width=0
       color {colorUseRule=UseColor_Specified, specifiedColor=<rgba 0 0 0 0>}
       cornerRadii=<0 0 0 0>
      }
      tintColor {colorUseRule=UseColor_Specified, specifiedColor=<rgba 0.0513 0.0953 0.159 1>}
      uVRegion {bIsValid=false, max=<0 0>, min=<0 0>}
     }
     oddRowBackgroundBrush {
      bIsDynamicallyLoaded=false
      drawAs=Image
      imageType=NoImage
      mirroring=NoMirror
      resourceName=None
      resourceObject=null
      tiling=NoTile
      imageSize=<0 0>
      margin=<ltrb 0 0 0 0>
      outlineSettings {
       bUseBrushTransparency=false
       roundingType=HalfHeightRadius
       width=0
       color {colorUseRule=UseColor_Specified, specifiedColor=<rgba 0 0 0 0>}
       cornerRadii=<0 0 0 0>
      }
      tintColor {colorUseRule=UseColor_Specified, specifiedColor=<rgba 0.0103 0.0103 0.0103 1>}
      uVRegion {bIsValid=false, max=<0 0>, min=<0 0>}
     }
     oddRowBackgroundHoveredBrush {
      bIsDynamicallyLoaded=false
      drawAs=Image
      imageType=NoImage
      mirroring=NoMirror
      resourceName=None
      resourceObject=null
      tiling=NoTile
      imageSize=<0 0>
      margin=<ltrb 0 0 0 0>
      outlineSettings {
       bUseBrushTransparency=false
       roundingType=HalfHeightRadius
       width=0
       color {colorUseRule=UseColor_Specified, specifiedColor=<rgba 0 0 0 0>}
       cornerRadii=<0 0 0 0>
      }
      tintColor {colorUseRule=UseColor_Specified, specifiedColor=<rgba 0.0176 0.0176 0.0176 1>}
      uVRegion {bIsValid=false, max=<0 0>, min=<0 0>}
     }
     parentRowBackgroundBrush {
      bIsDynamicallyLoaded=false
      drawAs=Image
      imageType=NoImage
      mirroring=NoMirror
      resourceName=None
      resourceObject=null
      tiling=NoTile
      imageSize=<32 32>
      margin=<ltrb 0 0 0 0>
      outlineSettings {
       bUseBrushTransparency=false
       roundingType=HalfHeightRadius
       width=0
       color {colorUseRule=UseColor_Specified, specifiedColor=<rgba 0 0 0 0>}
       cornerRadii=<0 0 0 0>
      }
      tintColor {colorUseRule=UseColor_Specified, specifiedColor=<rgba 1 1 1 1>}
      uVRegion {bIsValid=false, max=<0 0>, min=<0 0>}
     }
     parentRowBackgroundHoveredBrush {
      bIsDynamicallyLoaded=false
      drawAs=Image
      imageType=NoImage
      mirroring=NoMirror
      resourceName=None
      resourceObject=null
      tiling=NoTile
      imageSize=<32 32>
      margin=<ltrb 0 0 0 0>
      outlineSettings {
       bUseBrushTransparency=false
       roundingType=HalfHeightRadius
       width=0
       color {colorUseRule=UseColor_Specified, specifiedColor=<rgba 0 0 0 0>}
       cornerRadii=<0 0 0 0>
      }
      tintColor {colorUseRule=UseColor_Specified, specifiedColor=<rgba 1 1 1 1>}
      uVRegion {bIsValid=false, max=<0 0>, min=<0 0>}
     }
     selectedTextColor {colorUseRule=UseColor_Specified, specifiedColor=<rgba 1 1 1 1>}
     selectorFocusedBrush {
      bIsDynamicallyLoaded=false
      drawAs=Border
      imageType=FullColor
      mirroring=NoMirror
      resourceName=../../../Engine/Content/Slate/Common/Selector.png
      resourceObject=null
      tiling=Both
      imageSize=<0 0>
      margin=<ltrb 0.25 0.25 0.25 0.25>
      outlineSettings {
       bUseBrushTransparency=false
       roundingType=HalfHeightRadius
       width=0
       color {colorUseRule=UseColor_Specified, specifiedColor=<rgba 0 0 0 0>}
       cornerRadii=<0 0 0 0>
      }
      tintColor {colorUseRule=UseColor_Specified, specifiedColor=<rgba 0.701 0.225 0.003 1>}
      uVRegion {bIsValid=false, max=<0 0>, min=<0 0>}
     }
     textColor {colorUseRule=UseColor_Specified, specifiedColor=<rgba 0.5271 0.5271 0.5271 1>}
    }
    widgetStyle {
     comboButtonStyle {
      downArrowAlign=VAlign_Center
      buttonStyle {
       clickedSlateSound {resourceObject=null}
       disabled {
        bIsDynamicallyLoaded=false
        drawAs=RoundedBox
        imageType=NoImage
        mirroring=NoMirror
        resourceName=None
        resourceObject=null
        tiling=NoTile
        imageSize=<0 0>
        margin=<ltrb 0 0 0 0>
        outlineSettings {
         bUseBrushTransparency=false
         roundingType=FixedRadius
         width=1
         color {colorUseRule=UseColor_ColorTable, specifiedColor=<rgba 0 0 0 0>}
         cornerRadii=<4 4 4 4>
        }
        tintColor {colorUseRule=UseColor_Specified, specifiedColor=<rgba 0.0395 0.0395 0.0395 1>}
        uVRegion {bIsValid=false, max=<0 0>, min=<0 0>}
       }
       disabledForeground {colorUseRule=UseColor_ColorTable, specifiedColor=<rgba 0 0 0 0>}
       hovered {
        bIsDynamicallyLoaded=false
        drawAs=RoundedBox
        imageType=NoImage
        mirroring=NoMirror
        resourceName=None
        resourceObject=null
        tiling=NoTile
        imageSize=<0 0>
        margin=<ltrb 0 0 0 0>
        outlineSettings {
         bUseBrushTransparency=false
         roundingType=FixedRadius
         width=1
         color {colorUseRule=UseColor_ColorTable, specifiedColor=<rgba 0 0 0 0>}
         cornerRadii=<4 4 4 4>
        }
        tintColor {colorUseRule=UseColor_Specified, specifiedColor=<rgba 0.0048 0.0048 0.0048 1>}
        uVRegion {bIsValid=false, max=<0 0>, min=<0 0>}
       }
       hoveredForeground {colorUseRule=UseColor_ColorTable, specifiedColor=<rgba 0 0 0 0>}
       hoveredSlateSound {resourceObject=null}
       normal {
        bIsDynamicallyLoaded=false
        drawAs=RoundedBox
        imageType=NoImage
        mirroring=NoMirror
        resourceName=None
        resourceObject=null
        tiling=NoTile
        imageSize=<0 0>
        margin=<ltrb 0 0 0 0>
        outlineSettings {
         bUseBrushTransparency=false
         roundingType=FixedRadius
         width=1
         color {colorUseRule=UseColor_ColorTable, specifiedColor=<rgba 0 0 0 0>}
         cornerRadii=<4 4 4 4>
        }
        tintColor {colorUseRule=UseColor_Specified, specifiedColor=<rgba 0.0048 0.0048 0.0048 1>}
        uVRegion {bIsValid=false, max=<0 0>, min=<0 0>}
       }
       normalForeground {colorUseRule=UseColor_ColorTable, specifiedColor=<rgba 0 0 0 0>}
       normalPadding=<ltrb 8 1 8 1>
       pressed {
        bIsDynamicallyLoaded=false
        drawAs=RoundedBox
        imageType=NoImage
        mirroring=NoMirror
        resourceName=None
        resourceObject=null
        tiling=NoTile
        imageSize=<0 0>
        margin=<ltrb 0 0 0 0>
        outlineSettings {
         bUseBrushTransparency=false
         roundingType=FixedRadius
         width=1
         color {colorUseRule=UseColor_ColorTable, specifiedColor=<rgba 0 0 0 0>}
         cornerRadii=<4 4 4 4>
        }
        tintColor {colorUseRule=UseColor_Specified, specifiedColor=<rgba 0.0048 0.0048 0.0048 1>}
        uVRegion {bIsValid=false, max=<0 0>, min=<0 0>}
       }
       pressedForeground {colorUseRule=UseColor_ColorTable, specifiedColor=<rgba 0 0 0 0>}
       pressedPadding=<ltrb 8 2 8 0>
       pressedSlateSound {resourceObject=null}
      }
      contentPadding=<ltrb 0 0 0 0>
      downArrowImage {
       bIsDynamicallyLoaded=false
       drawAs=Image
       imageType=Vector
       mirroring=NoMirror
       resourceName=../../../Engine/Content/Slate/Starship/CoreWidgets/ComboBox/wide-chevron-down.svg
       resourceObject=null
       tiling=NoTile
       imageSize=<16 16>
       margin=<ltrb 0 0 0 0>
       outlineSettings {
        bUseBrushTransparency=false
        roundingType=HalfHeightRadius
        width=0
        color {colorUseRule=UseColor_Specified, specifiedColor=<rgba 0 0 0 0>}
        cornerRadii=<0 0 0 0>
       }
       tintColor {colorUseRule=UseColor_Specified, specifiedColor=<rgba 1 1 1 1>}
       uVRegion {bIsValid=false, max=<0 0>, min=<0 0>}
      }
      downArrowPadding=<ltrb 2 2 2 2>
      menuBorderBrush {
       bIsDynamicallyLoaded=false
       drawAs=RoundedBox
       imageType=NoImage
       mirroring=NoMirror
       resourceName=None
       resourceObject=null
       tiling=NoTile
       imageSize=<0 0>
       margin=<ltrb 0 0 0 0>
       outlineSettings {
        bUseBrushTransparency=false
        roundingType=FixedRadius
        width=1
        color {colorUseRule=UseColor_ColorTable, specifiedColor=<rgba 0 0 0 0>}
        cornerRadii=<0 0 0 0>
       }
       tintColor {colorUseRule=UseColor_Specified, specifiedColor=<rgba 0.0395 0.0395 0.0395 1>}
       uVRegion {bIsValid=false, max=<0 0>, min=<0 0>}
      }
      menuBorderPadding=<ltrb 0 0 0 0>
      shadowColorAndOpacity=<rgba 0 0 0 1>
      shadowOffset=<0 0>
     }
     contentPadding=<ltrb 0 0 0 0>
     menuRowPadding=<ltrb 8 3 8 3>
     pressedSlateSound {resourceObject=null}
     selectionChangeSlateSound {resourceObject=null}
    }
   }
   slot {class=HorizontalBoxSlot, properties {padding=<ltrb 5 5 5 5>}}
  }
  HB_ModKit_Options {
   widgetClassPath=&HorizontalBox
   bIsVariable=true
   parent=VerticalBox_0
   widget=HB_ModKit_Options
   slot {class=VerticalBoxSlot, properties {padding=<ltrb 5 5 5 5>}}
  }
  CommonTextBlock_93 {
   widgetClassPath=&CommonTextBlock
   parent=HB_ModKit_Options
   widget=CommonTextBlock_93
   properties {
    style=&TextStyle-Base_C
    text="Architecture Kit: "
    font {
     bForceMonospaced=false
     bMaterialIsStencil=false
     fontMaterial=null
     fontObject=&NotoSans
     letterSpacing=0
     monospacedWidth=1
     size=24
     skewAmount=0
     typefaceFontName=NotoSans-Regular
     outlineSettings {
      bApplyOutlineToDropShadows=false
      bMiteredCorners=false
      bSeparateFillAlpha=false
      outlineMaterial=null
      outlineSize=0
      outlineColor=<rgba 0 0 0 1>
     }
    }
    shadowOffset=<0 0>
   }
   slot {class=HorizontalBoxSlot, properties {padding=<ltrb 5 5 5 5>}}
  }
  ddKitOptions {
   widgetClassPath=&SearchableStringComboBox
   bIsVariable=true
   parent=HB_ModKit_Options
   widget=ddKitOptions
   properties {
    selectedOption=Core
    defaultOptions [Core,Expanded,Full]
    foregroundColor {colorUseRule=UseColor_Specified, specifiedColor=<rgba 0.5271 0.5271 0.5271 1>}
    itemStyle {
     bUseParentRowBrush=false
     activeBrush {
      bIsDynamicallyLoaded=false
      drawAs=Image
      imageType=NoImage
      mirroring=NoMirror
      resourceName=None
      resourceObject=null
      tiling=NoTile
      imageSize=<0 0>
      margin=<ltrb 0 0 0 0>
      outlineSettings {
       bUseBrushTransparency=false
       roundingType=HalfHeightRadius
       width=0
       color {colorUseRule=UseColor_Specified, specifiedColor=<rgba 0 0 0 0>}
       cornerRadii=<0 0 0 0>
      }
      tintColor {colorUseRule=UseColor_Specified, specifiedColor=<rgba 0 0.162 0.7454 1>}
      uVRegion {bIsValid=false, max=<0 0>, min=<0 0>}
     }
     activeHighlightedBrush {
      bIsDynamicallyLoaded=false
      drawAs=Image
      imageType=NoImage
      mirroring=NoMirror
      resourceName=None
      resourceObject=null
      tiling=NoTile
      imageSize=<0 0>
      margin=<ltrb 0 0 0 0>
      outlineSettings {
       bUseBrushTransparency=false
       roundingType=HalfHeightRadius
       width=0
       color {colorUseRule=UseColor_Specified, specifiedColor=<rgba 0 0 0 0>}
       cornerRadii=<0 0 0 0>
      }
      tintColor {colorUseRule=UseColor_Specified, specifiedColor=<rgba 0.0252 0.0319 0.0423 1>}
      uVRegion {bIsValid=false, max=<0 0>, min=<0 0>}
     }
     activeHoveredBrush {
      bIsDynamicallyLoaded=false
      drawAs=Image
      imageType=NoImage
      mirroring=NoMirror
      resourceName=None
      resourceObject=null
      tiling=NoTile
      imageSize=<0 0>
      margin=<ltrb 0 0 0 0>
      outlineSettings {
       bUseBrushTransparency=false
       roundingType=HalfHeightRadius
       width=0
       color {colorUseRule=UseColor_Specified, specifiedColor=<rgba 0 0 0 0>}
       cornerRadii=<0 0 0 0>
      }
      tintColor {colorUseRule=UseColor_Specified, specifiedColor=<rgba 0 0.162 0.7454 1>}
      uVRegion {bIsValid=false, max=<0 0>, min=<0 0>}
     }
     dropIndicator_Above {
      bIsDynamicallyLoaded=false
      drawAs=Box
      imageType=FullColor
      mirroring=NoMirror
      resourceName=../../../Engine/Content/Slate/Common/DropZoneIndicator_Above.png
      resourceObject=null
      tiling=NoTile
      imageSize=<0 0>
      margin=<ltrb 0.625 0.625 0 0>
      outlineSettings {
       bUseBrushTransparency=false
       roundingType=HalfHeightRadius
       width=0
       color {colorUseRule=UseColor_Specified, specifiedColor=<rgba 0 0 0 0>}
       cornerRadii=<0 0 0 0>
      }
      tintColor {colorUseRule=UseColor_Specified, specifiedColor=<rgba 0.0091 0.3515 0.9647 1>}
      uVRegion {bIsValid=false, max=<0 0>, min=<0 0>}
     }
     dropIndicator_Below {
      bIsDynamicallyLoaded=false
      drawAs=Box
      imageType=FullColor
      mirroring=NoMirror
      resourceName=../../../Engine/Content/Slate/Common/DropZoneIndicator_Below.png
      resourceObject=null
      tiling=NoTile
      imageSize=<0 0>
      margin=<ltrb 0.625 0 0 0.625>
      outlineSettings {
       bUseBrushTransparency=false
       roundingType=HalfHeightRadius
       width=0
       color {colorUseRule=UseColor_Specified, specifiedColor=<rgba 0 0 0 0>}
       cornerRadii=<0 0 0 0>
      }
      tintColor {colorUseRule=UseColor_Specified, specifiedColor=<rgba 0.0091 0.3515 0.9647 1>}
      uVRegion {bIsValid=false, max=<0 0>, min=<0 0>}
     }
     dropIndicator_Onto {
      bIsDynamicallyLoaded=false
      drawAs=Box
      imageType=FullColor
      mirroring=NoMirror
      resourceName=../../../Engine/Content/Slate/Common/DropZoneIndicator_Onto.png
      resourceObject=null
      tiling=NoTile
      imageSize=<0 0>
      margin=<ltrb 0.25 0.25 0.25 0.25>
      outlineSettings {
       bUseBrushTransparency=false
       roundingType=HalfHeightRadius
       width=0
       color {colorUseRule=UseColor_Specified, specifiedColor=<rgba 0 0 0 0>}
       cornerRadii=<0 0 0 0>
      }
      tintColor {colorUseRule=UseColor_Specified, specifiedColor=<rgba 0.0091 0.3515 0.9647 1>}
      uVRegion {bIsValid=false, max=<0 0>, min=<0 0>}
     }
     evenRowBackgroundBrush {
      bIsDynamicallyLoaded=false
      drawAs=Image
      imageType=NoImage
      mirroring=NoMirror
      resourceName=None
      resourceObject=null
      tiling=NoTile
      imageSize=<0 0>
      margin=<ltrb 0 0 0 0>
      outlineSettings {
       bUseBrushTransparency=false
       roundingType=HalfHeightRadius
       width=0
       color {colorUseRule=UseColor_Specified, specifiedColor=<rgba 0 0 0 0>}
       cornerRadii=<0 0 0 0>
      }
      tintColor {colorUseRule=UseColor_Specified, specifiedColor=<rgba 0.0103 0.0103 0.0103 1>}
      uVRegion {bIsValid=false, max=<0 0>, min=<0 0>}
     }
     evenRowBackgroundHoveredBrush {
      bIsDynamicallyLoaded=false
      drawAs=Image
      imageType=NoImage
      mirroring=NoMirror
      resourceName=None
      resourceObject=null
      tiling=NoTile
      imageSize=<0 0>
      margin=<ltrb 0 0 0 0>
      outlineSettings {
       bUseBrushTransparency=false
       roundingType=HalfHeightRadius
       width=0
       color {colorUseRule=UseColor_Specified, specifiedColor=<rgba 0 0 0 0>}
       cornerRadii=<0 0 0 0>
      }
      tintColor {colorUseRule=UseColor_Specified, specifiedColor=<rgba 0.0176 0.0176 0.0176 1>}
      uVRegion {bIsValid=false, max=<0 0>, min=<0 0>}
     }
     inactiveBrush {
      bIsDynamicallyLoaded=false
      drawAs=Image
      imageType=NoImage
      mirroring=NoMirror
      resourceName=None
      resourceObject=null
      tiling=NoTile
      imageSize=<0 0>
      margin=<ltrb 0 0 0 0>
      outlineSettings {
       bUseBrushTransparency=false
       roundingType=HalfHeightRadius
       width=0
       color {colorUseRule=UseColor_Specified, specifiedColor=<rgba 0 0 0 0>}
       cornerRadii=<0 0 0 0>
      }
      tintColor {colorUseRule=UseColor_Specified, specifiedColor=<rgba 0.0513 0.0953 0.159 1>}
      uVRegion {bIsValid=false, max=<0 0>, min=<0 0>}
     }
     inactiveHighlightedBrush {
      bIsDynamicallyLoaded=false
      drawAs=Image
      imageType=NoImage
      mirroring=NoMirror
      resourceName=None
      resourceObject=null
      tiling=NoTile
      imageSize=<0 0>
      margin=<ltrb 0 0 0 0>
      outlineSettings {
       bUseBrushTransparency=false
       roundingType=HalfHeightRadius
       width=0
       color {colorUseRule=UseColor_Specified, specifiedColor=<rgba 0 0 0 0>}
       cornerRadii=<0 0 0 0>
      }
      tintColor {colorUseRule=UseColor_Specified, specifiedColor=<rgba 0.0252 0.0319 0.0423 1>}
      uVRegion {bIsValid=false, max=<0 0>, min=<0 0>}
     }
     inactiveHoveredBrush {
      bIsDynamicallyLoaded=false
      drawAs=Image
      imageType=NoImage
      mirroring=NoMirror
      resourceName=None
      resourceObject=null
      tiling=NoTile
      imageSize=<0 0>
      margin=<ltrb 0 0 0 0>
      outlineSettings {
       bUseBrushTransparency=false
       roundingType=HalfHeightRadius
       width=0
       color {colorUseRule=UseColor_Specified, specifiedColor=<rgba 0 0 0 0>}
       cornerRadii=<0 0 0 0>
      }
      tintColor {colorUseRule=UseColor_Specified, specifiedColor=<rgba 0.0513 0.0953 0.159 1>}
      uVRegion {bIsValid=false, max=<0 0>, min=<0 0>}
     }
     oddRowBackgroundBrush {
      bIsDynamicallyLoaded=false
      drawAs=Image
      imageType=NoImage
      mirroring=NoMirror
      resourceName=None
      resourceObject=null
      tiling=NoTile
      imageSize=<0 0>
      margin=<ltrb 0 0 0 0>
      outlineSettings {
       bUseBrushTransparency=false
       roundingType=HalfHeightRadius
       width=0
       color {colorUseRule=UseColor_Specified, specifiedColor=<rgba 0 0 0 0>}
       cornerRadii=<0 0 0 0>
      }
      tintColor {colorUseRule=UseColor_Specified, specifiedColor=<rgba 0.0103 0.0103 0.0103 1>}
      uVRegion {bIsValid=false, max=<0 0>, min=<0 0>}
     }
     oddRowBackgroundHoveredBrush {
      bIsDynamicallyLoaded=false
      drawAs=Image
      imageType=NoImage
      mirroring=NoMirror
      resourceName=None
      resourceObject=null
      tiling=NoTile
      imageSize=<0 0>
      margin=<ltrb 0 0 0 0>
      outlineSettings {
       bUseBrushTransparency=false
       roundingType=HalfHeightRadius
       width=0
       color {colorUseRule=UseColor_Specified, specifiedColor=<rgba 0 0 0 0>}
       cornerRadii=<0 0 0 0>
      }
      tintColor {colorUseRule=UseColor_Specified, specifiedColor=<rgba 0.0176 0.0176 0.0176 1>}
      uVRegion {bIsValid=false, max=<0 0>, min=<0 0>}
     }
     parentRowBackgroundBrush {
      bIsDynamicallyLoaded=false
      drawAs=Image
      imageType=NoImage
      mirroring=NoMirror
      resourceName=None
      resourceObject=null
      tiling=NoTile
      imageSize=<32 32>
      margin=<ltrb 0 0 0 0>
      outlineSettings {
       bUseBrushTransparency=false
       roundingType=HalfHeightRadius
       width=0
       color {colorUseRule=UseColor_Specified, specifiedColor=<rgba 0 0 0 0>}
       cornerRadii=<0 0 0 0>
      }
      tintColor {colorUseRule=UseColor_Specified, specifiedColor=<rgba 1 1 1 1>}
      uVRegion {bIsValid=false, max=<0 0>, min=<0 0>}
     }
     parentRowBackgroundHoveredBrush {
      bIsDynamicallyLoaded=false
      drawAs=Image
      imageType=NoImage
      mirroring=NoMirror
      resourceName=None
      resourceObject=null
      tiling=NoTile
      imageSize=<32 32>
      margin=<ltrb 0 0 0 0>
      outlineSettings {
       bUseBrushTransparency=false
       roundingType=HalfHeightRadius
       width=0
       color {colorUseRule=UseColor_Specified, specifiedColor=<rgba 0 0 0 0>}
       cornerRadii=<0 0 0 0>
      }
      tintColor {colorUseRule=UseColor_Specified, specifiedColor=<rgba 1 1 1 1>}
      uVRegion {bIsValid=false, max=<0 0>, min=<0 0>}
     }
     selectedTextColor {colorUseRule=UseColor_Specified, specifiedColor=<rgba 1 1 1 1>}
     selectorFocusedBrush {
      bIsDynamicallyLoaded=false
      drawAs=Border
      imageType=FullColor
      mirroring=NoMirror
      resourceName=../../../Engine/Content/Slate/Common/Selector.png
      resourceObject=null
      tiling=Both
      imageSize=<0 0>
      margin=<ltrb 0.25 0.25 0.25 0.25>
      outlineSettings {
       bUseBrushTransparency=false
       roundingType=HalfHeightRadius
       width=0
       color {colorUseRule=UseColor_Specified, specifiedColor=<rgba 0 0 0 0>}
       cornerRadii=<0 0 0 0>
      }
      tintColor {colorUseRule=UseColor_Specified, specifiedColor=<rgba 0.701 0.225 0.003 1>}
      uVRegion {bIsValid=false, max=<0 0>, min=<0 0>}
     }
     textColor {colorUseRule=UseColor_Specified, specifiedColor=<rgba 0.5271 0.5271 0.5271 1>}
    }
    widgetStyle {
     comboButtonStyle {
      downArrowAlign=VAlign_Center
      buttonStyle {
       clickedSlateSound {resourceObject=null}
       disabled {
        bIsDynamicallyLoaded=false
        drawAs=RoundedBox
        imageType=NoImage
        mirroring=NoMirror
        resourceName=None
        resourceObject=null
        tiling=NoTile
        imageSize=<0 0>
        margin=<ltrb 0 0 0 0>
        outlineSettings {
         bUseBrushTransparency=false
         roundingType=FixedRadius
         width=1
         color {colorUseRule=UseColor_ColorTable, specifiedColor=<rgba 0 0 0 0>}
         cornerRadii=<4 4 4 4>
        }
        tintColor {colorUseRule=UseColor_Specified, specifiedColor=<rgba 0.0395 0.0395 0.0395 1>}
        uVRegion {bIsValid=false, max=<0 0>, min=<0 0>}
       }
       disabledForeground {colorUseRule=UseColor_ColorTable, specifiedColor=<rgba 0 0 0 0>}
       hovered {
        bIsDynamicallyLoaded=false
        drawAs=RoundedBox
        imageType=NoImage
        mirroring=NoMirror
        resourceName=None
        resourceObject=null
        tiling=NoTile
        imageSize=<0 0>
        margin=<ltrb 0 0 0 0>
        outlineSettings {
         bUseBrushTransparency=false
         roundingType=FixedRadius
         width=1
         color {colorUseRule=UseColor_ColorTable, specifiedColor=<rgba 0 0 0 0>}
         cornerRadii=<4 4 4 4>
        }
        tintColor {colorUseRule=UseColor_Specified, specifiedColor=<rgba 0.0048 0.0048 0.0048 1>}
        uVRegion {bIsValid=false, max=<0 0>, min=<0 0>}
       }
       hoveredForeground {colorUseRule=UseColor_ColorTable, specifiedColor=<rgba 0 0 0 0>}
       hoveredSlateSound {resourceObject=null}
       normal {
        bIsDynamicallyLoaded=false
        drawAs=RoundedBox
        imageType=NoImage
        mirroring=NoMirror
        resourceName=None
        resourceObject=null
        tiling=NoTile
        imageSize=<0 0>
        margin=<ltrb 0 0 0 0>
        outlineSettings {
         bUseBrushTransparency=false
         roundingType=FixedRadius
         width=1
         color {colorUseRule=UseColor_ColorTable, specifiedColor=<rgba 0 0 0 0>}
         cornerRadii=<4 4 4 4>
        }
        tintColor {colorUseRule=UseColor_Specified, specifiedColor=<rgba 0.0048 0.0048 0.0048 1>}
        uVRegion {bIsValid=false, max=<0 0>, min=<0 0>}
       }
       normalForeground {colorUseRule=UseColor_ColorTable, specifiedColor=<rgba 0 0 0 0>}
       normalPadding=<ltrb 8 1 8 1>
       pressed {
        bIsDynamicallyLoaded=false
        drawAs=RoundedBox
        imageType=NoImage
        mirroring=NoMirror
        resourceName=None
        resourceObject=null
        tiling=NoTile
        imageSize=<0 0>
        margin=<ltrb 0 0 0 0>
        outlineSettings {
         bUseBrushTransparency=false
         roundingType=FixedRadius
         width=1
         color {colorUseRule=UseColor_ColorTable, specifiedColor=<rgba 0 0 0 0>}
         cornerRadii=<4 4 4 4>
        }
        tintColor {colorUseRule=UseColor_Specified, specifiedColor=<rgba 0.0048 0.0048 0.0048 1>}
        uVRegion {bIsValid=false, max=<0 0>, min=<0 0>}
       }
       pressedForeground {colorUseRule=UseColor_ColorTable, specifiedColor=<rgba 0 0 0 0>}
       pressedPadding=<ltrb 8 2 8 0>
       pressedSlateSound {resourceObject=null}
      }
      contentPadding=<ltrb 0 0 0 0>
      downArrowImage {
       bIsDynamicallyLoaded=false
       drawAs=Image
       imageType=Vector
       mirroring=NoMirror
       resourceName=../../../Engine/Content/Slate/Starship/CoreWidgets/ComboBox/wide-chevron-down.svg
       resourceObject=null
       tiling=NoTile
       imageSize=<16 16>
       margin=<ltrb 0 0 0 0>
       outlineSettings {
        bUseBrushTransparency=false
        roundingType=HalfHeightRadius
        width=0
        color {colorUseRule=UseColor_Specified, specifiedColor=<rgba 0 0 0 0>}
        cornerRadii=<0 0 0 0>
       }
       tintColor {colorUseRule=UseColor_Specified, specifiedColor=<rgba 1 1 1 1>}
       uVRegion {bIsValid=false, max=<0 0>, min=<0 0>}
      }
      downArrowPadding=<ltrb 2 2 2 2>
      menuBorderBrush {
       bIsDynamicallyLoaded=false
       drawAs=RoundedBox
       imageType=NoImage
       mirroring=NoMirror
       resourceName=None
       resourceObject=null
       tiling=NoTile
       imageSize=<0 0>
       margin=<ltrb 0 0 0 0>
       outlineSettings {
        bUseBrushTransparency=false
        roundingType=FixedRadius
        width=1
        color {colorUseRule=UseColor_ColorTable, specifiedColor=<rgba 0 0 0 0>}
        cornerRadii=<0 0 0 0>
       }
       tintColor {colorUseRule=UseColor_Specified, specifiedColor=<rgba 0.0395 0.0395 0.0395 1>}
       uVRegion {bIsValid=false, max=<0 0>, min=<0 0>}
      }
      menuBorderPadding=<ltrb 0 0 0 0>
      shadowColorAndOpacity=<rgba 0 0 0 1>
      shadowOffset=<0 0>
     }
     contentPadding=<ltrb 0 0 0 0>
     menuRowPadding=<ltrb 8 3 8 3>
     pressedSlateSound {resourceObject=null}
     selectionChangeSlateSound {resourceObject=null}
    }
   }
   slot {class=HorizontalBoxSlot}
  }
  HB_PropParent {
   widgetClassPath=&HorizontalBox
   bIsVariable=true
   parent=VerticalBox_0
   widget=HB_PropParent
   slot {class=VerticalBoxSlot, properties {padding=<ltrb 10 10 10 10>}}
  }
  CommonTextBlock_94 {
   widgetClassPath=&CommonTextBlock
   parent=HB_PropParent
   widget=CommonTextBlock_94
   properties {
    style=&TextStyle-Base_C
    text=Parent Class:
    font {
     bForceMonospaced=false
     bMaterialIsStencil=false
     fontMaterial=null
     fontObject=&NotoSans
     letterSpacing=0
     monospacedWidth=1
     size=24
     skewAmount=0
     typefaceFontName=NotoSans-Regular
     outlineSettings {
      bApplyOutlineToDropShadows=false
      bMiteredCorners=false
      bSeparateFillAlpha=false
      outlineMaterial=null
      outlineSize=0
      outlineColor=<rgba 0 0 0 1>
     }
    }
    shadowOffset=<0 0>
   }
   slot {class=HorizontalBoxSlot, properties {padding=<ltrb 5 5 5 5>}}
  }
  ddParentClass {
   widgetClassPath=&SearchableStringComboBox
   bIsVariable=true
   parent=HB_PropParent
   widget=ddParentClass
   properties {
    selectedOption=Medium Prop
    defaultOptions [Small Prop, Medium Prop, Large Prop, Random Color Prop]
    foregroundColor {colorUseRule=UseColor_Specified, specifiedColor=<rgba 0.5271 0.5271 0.5271 1>}
    itemStyle {
     bUseParentRowBrush=false
     activeBrush {
      bIsDynamicallyLoaded=false
      drawAs=Image
      imageType=NoImage
      mirroring=NoMirror
      resourceName=None
      resourceObject=null
      tiling=NoTile
      imageSize=<0 0>
      margin=<ltrb 0 0 0 0>
      outlineSettings {
       bUseBrushTransparency=false
       roundingType=HalfHeightRadius
       width=0
       color {colorUseRule=UseColor_Specified, specifiedColor=<rgba 0 0 0 0>}
       cornerRadii=<0 0 0 0>
      }
      tintColor {colorUseRule=UseColor_Specified, specifiedColor=<rgba 0 0.162 0.7454 1>}
      uVRegion {bIsValid=false, max=<0 0>, min=<0 0>}
     }
     activeHighlightedBrush {
      bIsDynamicallyLoaded=false
      drawAs=Image
      imageType=NoImage
      mirroring=NoMirror
      resourceName=None
      resourceObject=null
      tiling=NoTile
      imageSize=<0 0>
      margin=<ltrb 0 0 0 0>
      outlineSettings {
       bUseBrushTransparency=false
       roundingType=HalfHeightRadius
       width=0
       color {colorUseRule=UseColor_Specified, specifiedColor=<rgba 0 0 0 0>}
       cornerRadii=<0 0 0 0>
      }
      tintColor {colorUseRule=UseColor_Specified, specifiedColor=<rgba 0.0252 0.0319 0.0423 1>}
      uVRegion {bIsValid=false, max=<0 0>, min=<0 0>}
     }
     activeHoveredBrush {
      bIsDynamicallyLoaded=false
      drawAs=Image
      imageType=NoImage
      mirroring=NoMirror
      resourceName=None
      resourceObject=null
      tiling=NoTile
      imageSize=<0 0>
      margin=<ltrb 0 0 0 0>
      outlineSettings {
       bUseBrushTransparency=false
       roundingType=HalfHeightRadius
       width=0
       color {colorUseRule=UseColor_Specified, specifiedColor=<rgba 0 0 0 0>}
       cornerRadii=<0 0 0 0>
      }
      tintColor {colorUseRule=UseColor_Specified, specifiedColor=<rgba 0 0.162 0.7454 1>}
      uVRegion {bIsValid=false, max=<0 0>, min=<0 0>}
     }
     dropIndicator_Above {
      bIsDynamicallyLoaded=false
      drawAs=Box
      imageType=FullColor
      mirroring=NoMirror
      resourceName=../../../Engine/Content/Slate/Common/DropZoneIndicator_Above.png
      resourceObject=null
      tiling=NoTile
      imageSize=<0 0>
      margin=<ltrb 0.625 0.625 0 0>
      outlineSettings {
       bUseBrushTransparency=false
       roundingType=HalfHeightRadius
       width=0
       color {colorUseRule=UseColor_Specified, specifiedColor=<rgba 0 0 0 0>}
       cornerRadii=<0 0 0 0>
      }
      tintColor {colorUseRule=UseColor_Specified, specifiedColor=<rgba 0.0091 0.3515 0.9647 1>}
      uVRegion {bIsValid=false, max=<0 0>, min=<0 0>}
     }
     dropIndicator_Below {
      bIsDynamicallyLoaded=false
      drawAs=Box
      imageType=FullColor
      mirroring=NoMirror
      resourceName=../../../Engine/Content/Slate/Common/DropZoneIndicator_Below.png
      resourceObject=null
      tiling=NoTile
      imageSize=<0 0>
      margin=<ltrb 0.625 0 0 0.625>
      outlineSettings {
       bUseBrushTransparency=false
       roundingType=HalfHeightRadius
       width=0
       color {colorUseRule=UseColor_Specified, specifiedColor=<rgba 0 0 0 0>}
       cornerRadii=<0 0 0 0>
      }
      tintColor {colorUseRule=UseColor_Specified, specifiedColor=<rgba 0.0091 0.3515 0.9647 1>}
      uVRegion {bIsValid=false, max=<0 0>, min=<0 0>}
     }
     dropIndicator_Onto {
      bIsDynamicallyLoaded=false
      drawAs=Box
      imageType=FullColor
      mirroring=NoMirror
      resourceName=../../../Engine/Content/Slate/Common/DropZoneIndicator_Onto.png
      resourceObject=null
      tiling=NoTile
      imageSize=<0 0>
      margin=<ltrb 0.25 0.25 0.25 0.25>
      outlineSettings {
       bUseBrushTransparency=false
       roundingType=HalfHeightRadius
       width=0
       color {colorUseRule=UseColor_Specified, specifiedColor=<rgba 0 0 0 0>}
       cornerRadii=<0 0 0 0>
      }
      tintColor {colorUseRule=UseColor_Specified, specifiedColor=<rgba 0.0091 0.3515 0.9647 1>}
      uVRegion {bIsValid=false, max=<0 0>, min=<0 0>}
     }
     evenRowBackgroundBrush {
      bIsDynamicallyLoaded=false
      drawAs=Image
      imageType=NoImage
      mirroring=NoMirror
      resourceName=None
      resourceObject=null
      tiling=NoTile
      imageSize=<0 0>
      margin=<ltrb 0 0 0 0>
      outlineSettings {
       bUseBrushTransparency=false
       roundingType=HalfHeightRadius
       width=0
       color {colorUseRule=UseColor_Specified, specifiedColor=<rgba 0 0 0 0>}
       cornerRadii=<0 0 0 0>
      }
      tintColor {colorUseRule=UseColor_Specified, specifiedColor=<rgba 0.0103 0.0103 0.0103 1>}
      uVRegion {bIsValid=false, max=<0 0>, min=<0 0>}
     }
     evenRowBackgroundHoveredBrush {
      bIsDynamicallyLoaded=false
      drawAs=Image
      imageType=NoImage
      mirroring=NoMirror
      resourceName=None
      resourceObject=null
      tiling=NoTile
      imageSize=<0 0>
      margin=<ltrb 0 0 0 0>
      outlineSettings {
       bUseBrushTransparency=false
       roundingType=HalfHeightRadius
       width=0
       color {colorUseRule=UseColor_Specified, specifiedColor=<rgba 0 0 0 0>}
       cornerRadii=<0 0 0 0>
      }
      tintColor {colorUseRule=UseColor_Specified, specifiedColor=<rgba 0.0176 0.0176 0.0176 1>}
      uVRegion {bIsValid=false, max=<0 0>, min=<0 0>}
     }
     inactiveBrush {
      bIsDynamicallyLoaded=false
      drawAs=Image
      imageType=NoImage
      mirroring=NoMirror
      resourceName=None
      resourceObject=null
      tiling=NoTile
      imageSize=<0 0>
      margin=<ltrb 0 0 0 0>
      outlineSettings {
       bUseBrushTransparency=false
       roundingType=HalfHeightRadius
       width=0
       color {colorUseRule=UseColor_Specified, specifiedColor=<rgba 0 0 0 0>}
       cornerRadii=<0 0 0 0>
      }
      tintColor {colorUseRule=UseColor_Specified, specifiedColor=<rgba 0.0513 0.0953 0.159 1>}
      uVRegion {bIsValid=false, max=<0 0>, min=<0 0>}
     }
     inactiveHighlightedBrush {
      bIsDynamicallyLoaded=false
      drawAs=Image
      imageType=NoImage
      mirroring=NoMirror
      resourceName=None
      resourceObject=null
      tiling=NoTile
      imageSize=<0 0>
      margin=<ltrb 0 0 0 0>
      outlineSettings {
       bUseBrushTransparency=false
       roundingType=HalfHeightRadius
       width=0
       color {colorUseRule=UseColor_Specified, specifiedColor=<rgba 0 0 0 0>}
       cornerRadii=<0 0 0 0>
      }
      tintColor {colorUseRule=UseColor_Specified, specifiedColor=<rgba 0.0252 0.0319 0.0423 1>}
      uVRegion {bIsValid=false, max=<0 0>, min=<0 0>}
     }
     inactiveHoveredBrush {
      bIsDynamicallyLoaded=false
      drawAs=Image
      imageType=NoImage
      mirroring=NoMirror
      resourceName=None
      resourceObject=null
      tiling=NoTile
      imageSize=<0 0>
      margin=<ltrb 0 0 0 0>
      outlineSettings {
       bUseBrushTransparency=false
       roundingType=HalfHeightRadius
       width=0
       color {colorUseRule=UseColor_Specified, specifiedColor=<rgba 0 0 0 0>}
       cornerRadii=<0 0 0 0>
      }
      tintColor {colorUseRule=UseColor_Specified, specifiedColor=<rgba 0.0513 0.0953 0.159 1>}
      uVRegion {bIsValid=false, max=<0 0>, min=<0 0>}
     }
     oddRowBackgroundBrush {
      bIsDynamicallyLoaded=false
      drawAs=Image
      imageType=NoImage
      mirroring=NoMirror
      resourceName=None
      resourceObject=null
      tiling=NoTile
      imageSize=<0 0>
      margin=<ltrb 0 0 0 0>
      outlineSettings {
       bUseBrushTransparency=false
       roundingType=HalfHeightRadius
       width=0
       color {colorUseRule=UseColor_Specified, specifiedColor=<rgba 0 0 0 0>}
       cornerRadii=<0 0 0 0>
      }
      tintColor {colorUseRule=UseColor_Specified, specifiedColor=<rgba 0.0103 0.0103 0.0103 1>}
      uVRegion {bIsValid=false, max=<0 0>, min=<0 0>}
     }
     oddRowBackgroundHoveredBrush {
      bIsDynamicallyLoaded=false
      drawAs=Image
      imageType=NoImage
      mirroring=NoMirror
      resourceName=None
      resourceObject=null
      tiling=NoTile
      imageSize=<0 0>
      margin=<ltrb 0 0 0 0>
      outlineSettings {
       bUseBrushTransparency=false
       roundingType=HalfHeightRadius
       width=0
       color {colorUseRule=UseColor_Specified, specifiedColor=<rgba 0 0 0 0>}
       cornerRadii=<0 0 0 0>
      }
      tintColor {colorUseRule=UseColor_Specified, specifiedColor=<rgba 0.0176 0.0176 0.0176 1>}
      uVRegion {bIsValid=false, max=<0 0>, min=<0 0>}
     }
     parentRowBackgroundBrush {
      bIsDynamicallyLoaded=false
      drawAs=Image
      imageType=NoImage
      mirroring=NoMirror
      resourceName=None
      resourceObject=null
      tiling=NoTile
      imageSize=<32 32>
      margin=<ltrb 0 0 0 0>
      outlineSettings {
       bUseBrushTransparency=false
       roundingType=HalfHeightRadius
       width=0
       color {colorUseRule=UseColor_Specified, specifiedColor=<rgba 0 0 0 0>}
       cornerRadii=<0 0 0 0>
      }
      tintColor {colorUseRule=UseColor_Specified, specifiedColor=<rgba 1 1 1 1>}
      uVRegion {bIsValid=false, max=<0 0>, min=<0 0>}
     }
     parentRowBackgroundHoveredBrush {
      bIsDynamicallyLoaded=false
      drawAs=Image
      imageType=NoImage
      mirroring=NoMirror
      resourceName=None
      resourceObject=null
      tiling=NoTile
      imageSize=<32 32>
      margin=<ltrb 0 0 0 0>
      outlineSettings {
       bUseBrushTransparency=false
       roundingType=HalfHeightRadius
       width=0
       color {colorUseRule=UseColor_Specified, specifiedColor=<rgba 0 0 0 0>}
       cornerRadii=<0 0 0 0>
      }
      tintColor {colorUseRule=UseColor_Specified, specifiedColor=<rgba 1 1 1 1>}
      uVRegion {bIsValid=false, max=<0 0>, min=<0 0>}
     }
     selectedTextColor {colorUseRule=UseColor_Specified, specifiedColor=<rgba 1 1 1 1>}
     selectorFocusedBrush {
      bIsDynamicallyLoaded=false
      drawAs=Border
      imageType=FullColor
      mirroring=NoMirror
      resourceName=../../../Engine/Content/Slate/Common/Selector.png
      resourceObject=null
      tiling=Both
      imageSize=<0 0>
      margin=<ltrb 0.25 0.25 0.25 0.25>
      outlineSettings {
       bUseBrushTransparency=false
       roundingType=HalfHeightRadius
       width=0
       color {colorUseRule=UseColor_Specified, specifiedColor=<rgba 0 0 0 0>}
       cornerRadii=<0 0 0 0>
      }
      tintColor {colorUseRule=UseColor_Specified, specifiedColor=<rgba 0.701 0.225 0.003 1>}
      uVRegion {bIsValid=false, max=<0 0>, min=<0 0>}
     }
     textColor {colorUseRule=UseColor_Specified, specifiedColor=<rgba 0.5271 0.5271 0.5271 1>}
    }
    widgetStyle {
     comboButtonStyle {
      downArrowAlign=VAlign_Center
      buttonStyle {
       clickedSlateSound {resourceObject=null}
       disabled {
        bIsDynamicallyLoaded=false
        drawAs=RoundedBox
        imageType=NoImage
        mirroring=NoMirror
        resourceName=None
        resourceObject=null
        tiling=NoTile
        imageSize=<0 0>
        margin=<ltrb 0 0 0 0>
        outlineSettings {
         bUseBrushTransparency=false
         roundingType=FixedRadius
         width=1
         color {colorUseRule=UseColor_ColorTable, specifiedColor=<rgba 0 0 0 0>}
         cornerRadii=<4 4 4 4>
        }
        tintColor {colorUseRule=UseColor_Specified, specifiedColor=<rgba 0.0395 0.0395 0.0395 1>}
        uVRegion {bIsValid=false, max=<0 0>, min=<0 0>}
       }
       disabledForeground {colorUseRule=UseColor_ColorTable, specifiedColor=<rgba 0 0 0 0>}
       hovered {
        bIsDynamicallyLoaded=false
        drawAs=RoundedBox
        imageType=NoImage
        mirroring=NoMirror
        resourceName=None
        resourceObject=null
        tiling=NoTile
        imageSize=<0 0>
        margin=<ltrb 0 0 0 0>
        outlineSettings {
         bUseBrushTransparency=false
         roundingType=FixedRadius
         width=1
         color {colorUseRule=UseColor_ColorTable, specifiedColor=<rgba 0 0 0 0>}
         cornerRadii=<4 4 4 4>
        }
        tintColor {colorUseRule=UseColor_Specified, specifiedColor=<rgba 0.0048 0.0048 0.0048 1>}
        uVRegion {bIsValid=false, max=<0 0>, min=<0 0>}
       }
       hoveredForeground {colorUseRule=UseColor_ColorTable, specifiedColor=<rgba 0 0 0 0>}
       hoveredSlateSound {resourceObject=null}
       normal {
        bIsDynamicallyLoaded=false
        drawAs=RoundedBox
        imageType=NoImage
        mirroring=NoMirror
        resourceName=None
        resourceObject=null
        tiling=NoTile
        imageSize=<0 0>
        margin=<ltrb 0 0 0 0>
        outlineSettings {
         bUseBrushTransparency=false
         roundingType=FixedRadius
         width=1
         color {colorUseRule=UseColor_ColorTable, specifiedColor=<rgba 0 0 0 0>}
         cornerRadii=<4 4 4 4>
        }
        tintColor {colorUseRule=UseColor_Specified, specifiedColor=<rgba 0.0048 0.0048 0.0048 1>}
        uVRegion {bIsValid=false, max=<0 0>, min=<0 0>}
       }
       normalForeground {colorUseRule=UseColor_ColorTable, specifiedColor=<rgba 0 0 0 0>}
       normalPadding=<ltrb 8 1 8 1>
       pressed {
        bIsDynamicallyLoaded=false
        drawAs=RoundedBox
        imageType=NoImage
        mirroring=NoMirror
        resourceName=None
        resourceObject=null
        tiling=NoTile
        imageSize=<0 0>
        margin=<ltrb 0 0 0 0>
        outlineSettings {
         bUseBrushTransparency=false
         roundingType=FixedRadius
         width=1
         color {colorUseRule=UseColor_ColorTable, specifiedColor=<rgba 0 0 0 0>}
         cornerRadii=<4 4 4 4>
        }
        tintColor {colorUseRule=UseColor_Specified, specifiedColor=<rgba 0.0048 0.0048 0.0048 1>}
        uVRegion {bIsValid=false, max=<0 0>, min=<0 0>}
       }
       pressedForeground {colorUseRule=UseColor_ColorTable, specifiedColor=<rgba 0 0 0 0>}
       pressedPadding=<ltrb 8 2 8 0>
       pressedSlateSound {resourceObject=null}
      }
      contentPadding=<ltrb 0 0 0 0>
      downArrowImage {
       bIsDynamicallyLoaded=false
       drawAs=Image
       imageType=Vector
       mirroring=NoMirror
       resourceName=../../../Engine/Content/Slate/Starship/CoreWidgets/ComboBox/wide-chevron-down.svg
       resourceObject=null
       tiling=NoTile
       imageSize=<16 16>
       margin=<ltrb 0 0 0 0>
       outlineSettings {
        bUseBrushTransparency=false
        roundingType=HalfHeightRadius
        width=0
        color {colorUseRule=UseColor_Specified, specifiedColor=<rgba 0 0 0 0>}
        cornerRadii=<0 0 0 0>
       }
       tintColor {colorUseRule=UseColor_Specified, specifiedColor=<rgba 1 1 1 1>}
       uVRegion {bIsValid=false, max=<0 0>, min=<0 0>}
      }
      downArrowPadding=<ltrb 2 2 2 2>
      menuBorderBrush {
       bIsDynamicallyLoaded=false
       drawAs=RoundedBox
       imageType=NoImage
       mirroring=NoMirror
       resourceName=None
       resourceObject=null
       tiling=NoTile
       imageSize=<0 0>
       margin=<ltrb 0 0 0 0>
       outlineSettings {
        bUseBrushTransparency=false
        roundingType=FixedRadius
        width=1
        color {colorUseRule=UseColor_ColorTable, specifiedColor=<rgba 0 0 0 0>}
        cornerRadii=<0 0 0 0>
       }
       tintColor {colorUseRule=UseColor_Specified, specifiedColor=<rgba 0.0395 0.0395 0.0395 1>}
       uVRegion {bIsValid=false, max=<0 0>, min=<0 0>}
      }
      menuBorderPadding=<ltrb 0 0 0 0>
      shadowColorAndOpacity=<rgba 0 0 0 1>
      shadowOffset=<0 0>
     }
     contentPadding=<ltrb 0 0 0 0>
     menuRowPadding=<ltrb 8 3 8 3>
     pressedSlateSound {resourceObject=null}
     selectionChangeSlateSound {resourceObject=null}
    }
   }
   slot {class=HorizontalBoxSlot, properties {padding=<ltrb 5 5 5 5>}}
  }
  HorizontalBox_5 {
   widgetClassPath=&HorizontalBox
   parent=VerticalBox_0
   widget=HorizontalBox_5
   slot {class=VerticalBoxSlot, properties {padding=<ltrb 10 0 0 0>}}
  }
  TXT_Category_Name {
   widgetClassPath=&TextBlock
   parent=HorizontalBox_5
   widget=TXT_Category_Name
   properties {
    justification=Center
    text=Category Name:
    font {
     bForceMonospaced=false
     bMaterialIsStencil=false
     fontMaterial=null
     fontObject=&Roboto
     letterSpacing=0
     monospacedWidth=1
     size=9
     skewAmount=0
     typefaceFontName=Bold
     outlineSettings {
      bApplyOutlineToDropShadows=false
      bMiteredCorners=false
      bSeparateFillAlpha=false
      outlineMaterial=null
      outlineSize=0
      outlineColor=<rgba 0 0 0 1>
     }
    }
   }
   slot {class=HorizontalBoxSlot, properties {padding=<ltrb 5 5 5 5>}}
  }
  TB_Category_Name {
   widgetClassPath=&EditableTextBox
   bIsVariable=true
   parent=HorizontalBox_5
   widget=TB_Category_Name
   properties {widgetStyle {
     backgroundColor {colorUseRule=UseColor_Specified, specifiedColor=<rgba 1 1 1 1>}
     backgroundImageFocused {
      bIsDynamicallyLoaded=false
      drawAs=RoundedBox
      imageType=NoImage
      mirroring=NoMirror
      resourceName=None
      resourceObject=null
      tiling=NoTile
      imageSize=<0 0>
      margin=<ltrb 0 0 0 0>
      outlineSettings {
       bUseBrushTransparency=false
       roundingType=FixedRadius
       width=1
       color {colorUseRule=UseColor_ColorTable, specifiedColor=<rgba 0 0 0 0>}
       cornerRadii=<4 4 4 4>
      }
      tintColor {colorUseRule=UseColor_Specified, specifiedColor=<rgba 0.0048 0.0048 0.0048 1>}
      uVRegion {bIsValid=false, max=<0 0>, min=<0 0>}
     }
     backgroundImageHovered {
      bIsDynamicallyLoaded=false
      drawAs=RoundedBox
      imageType=NoImage
      mirroring=NoMirror
      resourceName=None
      resourceObject=null
      tiling=NoTile
      imageSize=<0 0>
      margin=<ltrb 0 0 0 0>
      outlineSettings {
       bUseBrushTransparency=false
       roundingType=FixedRadius
       width=1
       color {colorUseRule=UseColor_ColorTable, specifiedColor=<rgba 0 0 0 0>}
       cornerRadii=<4 4 4 4>
      }
      tintColor {colorUseRule=UseColor_Specified, specifiedColor=<rgba 0.0048 0.0048 0.0048 1>}
      uVRegion {bIsValid=false, max=<0 0>, min=<0 0>}
     }
     backgroundImageNormal {
      bIsDynamicallyLoaded=false
      drawAs=RoundedBox
      imageType=NoImage
      mirroring=NoMirror
      resourceName=None
      resourceObject=null
      tiling=NoTile
      imageSize=<0 0>
      margin=<ltrb 0 0 0 0>
      outlineSettings {
       bUseBrushTransparency=false
       roundingType=FixedRadius
       width=1
       color {colorUseRule=UseColor_ColorTable, specifiedColor=<rgba 0 0 0 0>}
       cornerRadii=<4 4 4 4>
      }
      tintColor {colorUseRule=UseColor_Specified, specifiedColor=<rgba 0.0048 0.0048 0.0048 1>}
      uVRegion {bIsValid=false, max=<0 0>, min=<0 0>}
     }
     backgroundImageReadOnly {
      bIsDynamicallyLoaded=false
      drawAs=RoundedBox
      imageType=NoImage
      mirroring=NoMirror
      resourceName=None
      resourceObject=null
      tiling=NoTile
      imageSize=<0 0>
      margin=<ltrb 0 0 0 0>
      outlineSettings {
       bUseBrushTransparency=false
       roundingType=FixedRadius
       width=1
       color {colorUseRule=UseColor_ColorTable, specifiedColor=<rgba 0 0 0 0>}
       cornerRadii=<4 4 4 4>
      }
      tintColor {colorUseRule=UseColor_Specified, specifiedColor=<rgba 0.0284 0.0284 0.0284 1>}
      uVRegion {bIsValid=false, max=<0 0>, min=<0 0>}
     }
     focusedForegroundColor {colorUseRule=UseColor_Specified, specifiedColor=<rgba 1 1 1 1>}
     foregroundColor {colorUseRule=UseColor_Specified, specifiedColor=<rgba 0.5271 0.5271 0.5271 1>}
     hScrollBarPadding=<ltrb 0 0 0 0>
     padding=<ltrb 8 3 8 4>
     readOnlyForegroundColor {colorUseRule=UseColor_Foreground, specifiedColor=<rgba 1 0 1 1>}
     scrollBarStyle {
      thickness=8
      draggedThumbImage {
       bIsDynamicallyLoaded=false
       drawAs=RoundedBox
       imageType=NoImage
       mirroring=NoMirror
       resourceName=None
       resourceObject=null
       tiling=NoTile
       imageSize=<0 0>
       margin=<ltrb 0 0 0 0>
       outlineSettings {
        bUseBrushTransparency=false
        roundingType=FixedRadius
        width=0
        color {colorUseRule=UseColor_Specified, specifiedColor=<rgba 0 0 0 0>}
        cornerRadii=<4 4 4 4>
       }
       tintColor {colorUseRule=UseColor_Specified, specifiedColor=<rgba 0.2159 0.2159 0.2159 1>}
       uVRegion {bIsValid=false, max=<0 0>, min=<0 0>}
      }
      horizontalBackgroundImage {
       bIsDynamicallyLoaded=false
       drawAs=NoDrawType
       imageType=NoImage
       mirroring=NoMirror
       resourceName=None
       resourceObject=null
       tiling=NoTile
       imageSize=<0 0>
       margin=<ltrb 0 0 0 0>
       outlineSettings {
        bUseBrushTransparency=false
        roundingType=HalfHeightRadius
        width=0
        color {colorUseRule=UseColor_Specified, specifiedColor=<rgba 0 0 0 0>}
        cornerRadii=<0 0 0 0>
       }
       tintColor {colorUseRule=UseColor_Specified, specifiedColor=<rgba 1 1 1 1>}
       uVRegion {bIsValid=false, max=<0 0>, min=<0 0>}
      }
      horizontalBottomSlotImage {
       bIsDynamicallyLoaded=false
       drawAs=NoDrawType
       imageType=NoImage
       mirroring=NoMirror
       resourceName=None
       resourceObject=null
       tiling=NoTile
       imageSize=<0 0>
       margin=<ltrb 0 0 0 0>
       outlineSettings {
        bUseBrushTransparency=false
        roundingType=HalfHeightRadius
        width=0
        color {colorUseRule=UseColor_Specified, specifiedColor=<rgba 0 0 0 0>}
        cornerRadii=<0 0 0 0>
       }
       tintColor {colorUseRule=UseColor_Specified, specifiedColor=<rgba 1 1 1 1>}
       uVRegion {bIsValid=false, max=<0 0>, min=<0 0>}
      }
      horizontalTopSlotImage {
       bIsDynamicallyLoaded=false
       drawAs=NoDrawType
       imageType=NoImage
       mirroring=NoMirror
       resourceName=None
       resourceObject=null
       tiling=NoTile
       imageSize=<0 0>
       margin=<ltrb 0 0 0 0>
       outlineSettings {
        bUseBrushTransparency=false
        roundingType=HalfHeightRadius
        width=0
        color {colorUseRule=UseColor_Specified, specifiedColor=<rgba 0 0 0 0>}
        cornerRadii=<0 0 0 0>
       }
       tintColor {colorUseRule=UseColor_Specified, specifiedColor=<rgba 1 1 1 1>}
       uVRegion {bIsValid=false, max=<0 0>, min=<0 0>}
      }
      hoveredThumbImage {
       bIsDynamicallyLoaded=false
       drawAs=RoundedBox
       imageType=NoImage
       mirroring=NoMirror
       resourceName=None
       resourceObject=null
       tiling=NoTile
       imageSize=<0 0>
       margin=<ltrb 0 0 0 0>
       outlineSettings {
        bUseBrushTransparency=false
        roundingType=FixedRadius
        width=0
        color {colorUseRule=UseColor_Specified, specifiedColor=<rgba 0 0 0 0>}
        cornerRadii=<4 4 4 4>
       }
       tintColor {colorUseRule=UseColor_Specified, specifiedColor=<rgba 0.2159 0.2159 0.2159 1>}
       uVRegion {bIsValid=false, max=<0 0>, min=<0 0>}
      }
      normalThumbImage {
       bIsDynamicallyLoaded=false
       drawAs=RoundedBox
       imageType=NoImage
       mirroring=NoMirror
       resourceName=None
       resourceObject=null
       tiling=NoTile
       imageSize=<0 0>
       margin=<ltrb 0 0 0 0>
       outlineSettings {
        bUseBrushTransparency=false
        roundingType=FixedRadius
        width=0
        color {colorUseRule=UseColor_Specified, specifiedColor=<rgba 0 0 0 0>}
        cornerRadii=<4 4 4 4>
       }
       tintColor {colorUseRule=UseColor_Specified, specifiedColor=<rgba 0.0953 0.0953 0.0953 1>}
       uVRegion {bIsValid=false, max=<0 0>, min=<0 0>}
      }
      verticalBackgroundImage {
       bIsDynamicallyLoaded=false
       drawAs=NoDrawType
       imageType=NoImage
       mirroring=NoMirror
       resourceName=None
       resourceObject=null
       tiling=NoTile
       imageSize=<0 0>
       margin=<ltrb 0 0 0 0>
       outlineSettings {
        bUseBrushTransparency=false
        roundingType=HalfHeightRadius
        width=0
        color {colorUseRule=UseColor_Specified, specifiedColor=<rgba 0 0 0 0>}
        cornerRadii=<0 0 0 0>
       }
       tintColor {colorUseRule=UseColor_Specified, specifiedColor=<rgba 1 1 1 1>}
       uVRegion {bIsValid=false, max=<0 0>, min=<0 0>}
      }
      verticalBottomSlotImage {
       bIsDynamicallyLoaded=false
       drawAs=NoDrawType
       imageType=NoImage
       mirroring=NoMirror
       resourceName=None
       resourceObject=null
       tiling=NoTile
       imageSize=<0 0>
       margin=<ltrb 0 0 0 0>
       outlineSettings {
        bUseBrushTransparency=false
        roundingType=HalfHeightRadius
        width=0
        color {colorUseRule=UseColor_Specified, specifiedColor=<rgba 0 0 0 0>}
        cornerRadii=<0 0 0 0>
       }
       tintColor {colorUseRule=UseColor_Specified, specifiedColor=<rgba 1 1 1 1>}
       uVRegion {bIsValid=false, max=<0 0>, min=<0 0>}
      }
      verticalTopSlotImage {
       bIsDynamicallyLoaded=false
       drawAs=NoDrawType
       imageType=NoImage
       mirroring=NoMirror
       resourceName=None
       resourceObject=null
       tiling=NoTile
       imageSize=<0 0>
       margin=<ltrb 0 0 0 0>
       outlineSettings {
        bUseBrushTransparency=false
        roundingType=HalfHeightRadius
        width=0
        color {colorUseRule=UseColor_Specified, specifiedColor=<rgba 0 0 0 0>}
        cornerRadii=<0 0 0 0>
       }
       tintColor {colorUseRule=UseColor_Specified, specifiedColor=<rgba 1 1 1 1>}
       uVRegion {bIsValid=false, max=<0 0>, min=<0 0>}
      }
     }
     textStyle {
      overflowPolicy=Clip
      transformPolicy=None
      colorAndOpacity {colorUseRule=UseColor_Foreground, specifiedColor=<rgba 1 0 1 1>}
      font {
       bForceMonospaced=false
       bMaterialIsStencil=false
       fontMaterial=null
       fontObject=null
       letterSpacing=0
       monospacedWidth=1
       size=10
       skewAmount=0
       typefaceFontName=Regular
       outlineSettings {
        bApplyOutlineToDropShadows=false
        bMiteredCorners=false
        bSeparateFillAlpha=false
        outlineMaterial=null
        outlineSize=0
        outlineColor=<rgba 0 0 0 1>
       }
      }
      highlightColor {colorUseRule=UseColor_ColorTable, specifiedColor=<rgba 0 0 0 0>}
      highlightShape {
       bIsDynamicallyLoaded=false
       drawAs=Image
       imageType=NoImage
       mirroring=NoMirror
       resourceName=None
       resourceObject=null
       tiling=NoTile
       imageSize=<0 0>
       margin=<ltrb 0 0 0 0>
       outlineSettings {
        bUseBrushTransparency=false
        roundingType=HalfHeightRadius
        width=0
        color {colorUseRule=UseColor_Specified, specifiedColor=<rgba 0 0 0 0>}
        cornerRadii=<0 0 0 0>
       }
       tintColor {colorUseRule=UseColor_ColorTable, specifiedColor=<rgba 0 0 0 0>}
       uVRegion {bIsValid=false, max=<0 0>, min=<0 0>}
      }
      selectedBackgroundColor {colorUseRule=UseColor_ColorTable, specifiedColor=<rgba 0 0 0 0>}
      shadowColorAndOpacity=<rgba 0 0 0 1>
      shadowOffset=<0 0>
      strikeBrush {
       bIsDynamicallyLoaded=false
       drawAs=Image
       imageType=NoImage
       mirroring=NoMirror
       resourceName=None
       resourceObject=null
       tiling=NoTile
       imageSize=<32 32>
       margin=<ltrb 0 0 0 0>
       outlineSettings {
        bUseBrushTransparency=false
        roundingType=HalfHeightRadius
        width=0
        color {colorUseRule=UseColor_Specified, specifiedColor=<rgba 0 0 0 0>}
        cornerRadii=<0 0 0 0>
       }
       tintColor {colorUseRule=UseColor_Specified, specifiedColor=<rgba 1 1 1 1>}
       uVRegion {bIsValid=false, max=<0 0>, min=<0 0>}
      }
      underlineBrush {
       bIsDynamicallyLoaded=false
       drawAs=Image
       imageType=NoImage
       mirroring=NoMirror
       resourceName=None
       resourceObject=null
       tiling=NoTile
       imageSize=<32 32>
       margin=<ltrb 0 0 0 0>
       outlineSettings {
        bUseBrushTransparency=false
        roundingType=HalfHeightRadius
        width=0
        color {colorUseRule=UseColor_Specified, specifiedColor=<rgba 0 0 0 0>}
        cornerRadii=<0 0 0 0>
       }
       tintColor {colorUseRule=UseColor_Specified, specifiedColor=<rgba 1 1 1 1>}
       uVRegion {bIsValid=false, max=<0 0>, min=<0 0>}
      }
     }
     vScrollBarPadding=<ltrb 0 0 0 0>
    }}
   slot {class=HorizontalBoxSlot, properties {size {sizeRule=Fill, value=1}}}
  }
  HB_PropName {
   widgetClassPath=&HorizontalBox
   bIsVariable=true
   parent=VerticalBox_0
   widget=HB_PropName
   slot {class=VerticalBoxSlot, properties {padding=<ltrb 10 0 0 0>}}
  }
  TXT_AssetName {
   widgetClassPath=&TextBlock
   parent=HB_PropName
   widget=TXT_AssetName
   properties {
    justification=Center
    text=Asset Name:
    font {
     bForceMonospaced=false
     bMaterialIsStencil=false
     fontMaterial=null
     fontObject=&Roboto
     letterSpacing=0
     monospacedWidth=1
     size=9
     skewAmount=0
     typefaceFontName=Bold
     outlineSettings {
      bApplyOutlineToDropShadows=false
      bMiteredCorners=false
      bSeparateFillAlpha=false
      outlineMaterial=null
      outlineSize=0
      outlineColor=<rgba 0 0 0 1>
     }
    }
   }
   slot {class=HorizontalBoxSlot, properties {padding=<ltrb 5 5 5 5>}}
  }
  TB_AssetName {
   widgetClassPath=&EditableTextBox
   bIsVariable=true
   parent=HB_PropName
   widget=TB_AssetName
   properties {widgetStyle {
     backgroundColor {colorUseRule=UseColor_Specified, specifiedColor=<rgba 1 1 1 1>}
     backgroundImageFocused {
      bIsDynamicallyLoaded=false
      drawAs=RoundedBox
      imageType=NoImage
      mirroring=NoMirror
      resourceName=None
      resourceObject=null
      tiling=NoTile
      imageSize=<0 0>
      margin=<ltrb 0 0 0 0>
      outlineSettings {
       bUseBrushTransparency=false
       roundingType=FixedRadius
       width=1
       color {colorUseRule=UseColor_ColorTable, specifiedColor=<rgba 0 0 0 0>}
       cornerRadii=<4 4 4 4>
      }
      tintColor {colorUseRule=UseColor_Specified, specifiedColor=<rgba 0.0048 0.0048 0.0048 1>}
      uVRegion {bIsValid=false, max=<0 0>, min=<0 0>}
     }
     backgroundImageHovered {
      bIsDynamicallyLoaded=false
      drawAs=RoundedBox
      imageType=NoImage
      mirroring=NoMirror
      resourceName=None
      resourceObject=null
      tiling=NoTile
      imageSize=<0 0>
      margin=<ltrb 0 0 0 0>
      outlineSettings {
       bUseBrushTransparency=false
       roundingType=FixedRadius
       width=1
       color {colorUseRule=UseColor_ColorTable, specifiedColor=<rgba 0 0 0 0>}
       cornerRadii=<4 4 4 4>
      }
      tintColor {colorUseRule=UseColor_Specified, specifiedColor=<rgba 0.0048 0.0048 0.0048 1>}
      uVRegion {bIsValid=false, max=<0 0>, min=<0 0>}
     }
     backgroundImageNormal {
      bIsDynamicallyLoaded=false
      drawAs=RoundedBox
      imageType=NoImage
      mirroring=NoMirror
      resourceName=None
      resourceObject=null
      tiling=NoTile
      imageSize=<0 0>
      margin=<ltrb 0 0 0 0>
      outlineSettings {
       bUseBrushTransparency=false
       roundingType=FixedRadius
       width=1
       color {colorUseRule=UseColor_ColorTable, specifiedColor=<rgba 0 0 0 0>}
       cornerRadii=<4 4 4 4>
      }
      tintColor {colorUseRule=UseColor_Specified, specifiedColor=<rgba 0.0048 0.0048 0.0048 1>}
      uVRegion {bIsValid=false, max=<0 0>, min=<0 0>}
     }
     backgroundImageReadOnly {
      bIsDynamicallyLoaded=false
      drawAs=RoundedBox
      imageType=NoImage
      mirroring=NoMirror
      resourceName=None
      resourceObject=null
      tiling=NoTile
      imageSize=<0 0>
      margin=<ltrb 0 0 0 0>
      outlineSettings {
       bUseBrushTransparency=false
       roundingType=FixedRadius
       width=1
       color {colorUseRule=UseColor_ColorTable, specifiedColor=<rgba 0 0 0 0>}
       cornerRadii=<4 4 4 4>
      }
      tintColor {colorUseRule=UseColor_Specified, specifiedColor=<rgba 0.0284 0.0284 0.0284 1>}
      uVRegion {bIsValid=false, max=<0 0>, min=<0 0>}
     }
     focusedForegroundColor {colorUseRule=UseColor_Specified, specifiedColor=<rgba 1 1 1 1>}
     foregroundColor {colorUseRule=UseColor_Specified, specifiedColor=<rgba 0.5271 0.5271 0.5271 1>}
     hScrollBarPadding=<ltrb 0 0 0 0>
     padding=<ltrb 8 3 8 4>
     readOnlyForegroundColor {colorUseRule=UseColor_Foreground, specifiedColor=<rgba 1 0 1 1>}
     scrollBarStyle {
      thickness=8
      draggedThumbImage {
       bIsDynamicallyLoaded=false
       drawAs=RoundedBox
       imageType=NoImage
       mirroring=NoMirror
       resourceName=None
       resourceObject=null
       tiling=NoTile
       imageSize=<0 0>
       margin=<ltrb 0 0 0 0>
       outlineSettings {
        bUseBrushTransparency=false
        roundingType=FixedRadius
        width=0
        color {colorUseRule=UseColor_Specified, specifiedColor=<rgba 0 0 0 0>}
        cornerRadii=<4 4 4 4>
       }
       tintColor {colorUseRule=UseColor_Specified, specifiedColor=<rgba 0.2159 0.2159 0.2159 1>}
       uVRegion {bIsValid=false, max=<0 0>, min=<0 0>}
      }
      horizontalBackgroundImage {
       bIsDynamicallyLoaded=false
       drawAs=NoDrawType
       imageType=NoImage
       mirroring=NoMirror
       resourceName=None
       resourceObject=null
       tiling=NoTile
       imageSize=<0 0>
       margin=<ltrb 0 0 0 0>
       outlineSettings {
        bUseBrushTransparency=false
        roundingType=HalfHeightRadius
        width=0
        color {colorUseRule=UseColor_Specified, specifiedColor=<rgba 0 0 0 0>}
        cornerRadii=<0 0 0 0>
       }
       tintColor {colorUseRule=UseColor_Specified, specifiedColor=<rgba 1 1 1 1>}
       uVRegion {bIsValid=false, max=<0 0>, min=<0 0>}
      }
      horizontalBottomSlotImage {
       bIsDynamicallyLoaded=false
       drawAs=NoDrawType
       imageType=NoImage
       mirroring=NoMirror
       resourceName=None
       resourceObject=null
       tiling=NoTile
       imageSize=<0 0>
       margin=<ltrb 0 0 0 0>
       outlineSettings {
        bUseBrushTransparency=false
        roundingType=HalfHeightRadius
        width=0
        color {colorUseRule=UseColor_Specified, specifiedColor=<rgba 0 0 0 0>}
        cornerRadii=<0 0 0 0>
       }
       tintColor {colorUseRule=UseColor_Specified, specifiedColor=<rgba 1 1 1 1>}
       uVRegion {bIsValid=false, max=<0 0>, min=<0 0>}
      }
      horizontalTopSlotImage {
       bIsDynamicallyLoaded=false
       drawAs=NoDrawType
       imageType=NoImage
       mirroring=NoMirror
       resourceName=None
       resourceObject=null
       tiling=NoTile
       imageSize=<0 0>
       margin=<ltrb 0 0 0 0>
       outlineSettings {
        bUseBrushTransparency=false
        roundingType=HalfHeightRadius
        width=0
        color {colorUseRule=UseColor_Specified, specifiedColor=<rgba 0 0 0 0>}
        cornerRadii=<0 0 0 0>
       }
       tintColor {colorUseRule=UseColor_Specified, specifiedColor=<rgba 1 1 1 1>}
       uVRegion {bIsValid=false, max=<0 0>, min=<0 0>}
      }
      hoveredThumbImage {
       bIsDynamicallyLoaded=false
       drawAs=RoundedBox
       imageType=NoImage
       mirroring=NoMirror
       resourceName=None
       resourceObject=null
       tiling=NoTile
       imageSize=<0 0>
       margin=<ltrb 0 0 0 0>
       outlineSettings {
        bUseBrushTransparency=false
        roundingType=FixedRadius
        width=0
        color {colorUseRule=UseColor_Specified, specifiedColor=<rgba 0 0 0 0>}
        cornerRadii=<4 4 4 4>
       }
       tintColor {colorUseRule=UseColor_Specified, specifiedColor=<rgba 0.2159 0.2159 0.2159 1>}
       uVRegion {bIsValid=false, max=<0 0>, min=<0 0>}
      }
      normalThumbImage {
       bIsDynamicallyLoaded=false
       drawAs=RoundedBox
       imageType=NoImage
       mirroring=NoMirror
       resourceName=None
       resourceObject=null
       tiling=NoTile
       imageSize=<0 0>
       margin=<ltrb 0 0 0 0>
       outlineSettings {
        bUseBrushTransparency=false
        roundingType=FixedRadius
        width=0
        color {colorUseRule=UseColor_Specified, specifiedColor=<rgba 0 0 0 0>}
        cornerRadii=<4 4 4 4>
       }
       tintColor {colorUseRule=UseColor_Specified, specifiedColor=<rgba 0.0953 0.0953 0.0953 1>}
       uVRegion {bIsValid=false, max=<0 0>, min=<0 0>}
      }
      verticalBackgroundImage {
       bIsDynamicallyLoaded=false
       drawAs=NoDrawType
       imageType=NoImage
       mirroring=NoMirror
       resourceName=None
       resourceObject=null
       tiling=NoTile
       imageSize=<0 0>
       margin=<ltrb 0 0 0 0>
       outlineSettings {
        bUseBrushTransparency=false
        roundingType=HalfHeightRadius
        width=0
        color {colorUseRule=UseColor_Specified, specifiedColor=<rgba 0 0 0 0>}
        cornerRadii=<0 0 0 0>
       }
       tintColor {colorUseRule=UseColor_Specified, specifiedColor=<rgba 1 1 1 1>}
       uVRegion {bIsValid=false, max=<0 0>, min=<0 0>}
      }
      verticalBottomSlotImage {
       bIsDynamicallyLoaded=false
       drawAs=NoDrawType
       imageType=NoImage
       mirroring=NoMirror
       resourceName=None
       resourceObject=null
       tiling=NoTile
       imageSize=<0 0>
       margin=<ltrb 0 0 0 0>
       outlineSettings {
        bUseBrushTransparency=false
        roundingType=HalfHeightRadius
        width=0
        color {colorUseRule=UseColor_Specified, specifiedColor=<rgba 0 0 0 0>}
        cornerRadii=<0 0 0 0>
       }
       tintColor {colorUseRule=UseColor_Specified, specifiedColor=<rgba 1 1 1 1>}
       uVRegion {bIsValid=false, max=<0 0>, min=<0 0>}
      }
      verticalTopSlotImage {
       bIsDynamicallyLoaded=false
       drawAs=NoDrawType
       imageType=NoImage
       mirroring=NoMirror
       resourceName=None
       resourceObject=null
       tiling=NoTile
       imageSize=<0 0>
       margin=<ltrb 0 0 0 0>
       outlineSettings {
        bUseBrushTransparency=false
        roundingType=HalfHeightRadius
        width=0
        color {colorUseRule=UseColor_Specified, specifiedColor=<rgba 0 0 0 0>}
        cornerRadii=<0 0 0 0>
       }
       tintColor {colorUseRule=UseColor_Specified, specifiedColor=<rgba 1 1 1 1>}
       uVRegion {bIsValid=false, max=<0 0>, min=<0 0>}
      }
     }
     textStyle {
      overflowPolicy=Clip
      transformPolicy=None
      colorAndOpacity {colorUseRule=UseColor_Foreground, specifiedColor=<rgba 1 0 1 1>}
      font {
       bForceMonospaced=false
       bMaterialIsStencil=false
       fontMaterial=null
       fontObject=null
       letterSpacing=0
       monospacedWidth=1
       size=10
       skewAmount=0
       typefaceFontName=Regular
       outlineSettings {
        bApplyOutlineToDropShadows=false
        bMiteredCorners=false
        bSeparateFillAlpha=false
        outlineMaterial=null
        outlineSize=0
        outlineColor=<rgba 0 0 0 1>
       }
      }
      highlightColor {colorUseRule=UseColor_ColorTable, specifiedColor=<rgba 0 0 0 0>}
      highlightShape {
       bIsDynamicallyLoaded=false
       drawAs=Image
       imageType=NoImage
       mirroring=NoMirror
       resourceName=None
       resourceObject=null
       tiling=NoTile
       imageSize=<0 0>
       margin=<ltrb 0 0 0 0>
       outlineSettings {
        bUseBrushTransparency=false
        roundingType=HalfHeightRadius
        width=0
        color {colorUseRule=UseColor_Specified, specifiedColor=<rgba 0 0 0 0>}
        cornerRadii=<0 0 0 0>
       }
       tintColor {colorUseRule=UseColor_ColorTable, specifiedColor=<rgba 0 0 0 0>}
       uVRegion {bIsValid=false, max=<0 0>, min=<0 0>}
      }
      selectedBackgroundColor {colorUseRule=UseColor_ColorTable, specifiedColor=<rgba 0 0 0 0>}
      shadowColorAndOpacity=<rgba 0 0 0 1>
      shadowOffset=<0 0>
      strikeBrush {
       bIsDynamicallyLoaded=false
       drawAs=Image
       imageType=NoImage
       mirroring=NoMirror
       resourceName=None
       resourceObject=null
       tiling=NoTile
       imageSize=<32 32>
       margin=<ltrb 0 0 0 0>
       outlineSettings {
        bUseBrushTransparency=false
        roundingType=HalfHeightRadius
        width=0
        color {colorUseRule=UseColor_Specified, specifiedColor=<rgba 0 0 0 0>}
        cornerRadii=<0 0 0 0>
       }
       tintColor {colorUseRule=UseColor_Specified, specifiedColor=<rgba 1 1 1 1>}
       uVRegion {bIsValid=false, max=<0 0>, min=<0 0>}
      }
      underlineBrush {
       bIsDynamicallyLoaded=false
       drawAs=Image
       imageType=NoImage
       mirroring=NoMirror
       resourceName=None
       resourceObject=null
       tiling=NoTile
       imageSize=<32 32>
       margin=<ltrb 0 0 0 0>
       outlineSettings {
        bUseBrushTransparency=false
        roundingType=HalfHeightRadius
        width=0
        color {colorUseRule=UseColor_Specified, specifiedColor=<rgba 0 0 0 0>}
        cornerRadii=<0 0 0 0>
       }
       tintColor {colorUseRule=UseColor_Specified, specifiedColor=<rgba 1 1 1 1>}
       uVRegion {bIsValid=false, max=<0 0>, min=<0 0>}
      }
     }
     vScrollBarPadding=<ltrb 0 0 0 0>
    }}
   slot {class=HorizontalBoxSlot, properties {size {sizeRule=Fill, value=1}}}
  }
  HorizontalBox_2 {
   widgetClassPath=&HorizontalBox
   parent=VerticalBox_0
   widget=HorizontalBox_2
   slot {class=VerticalBoxSlot, properties {padding=<ltrb 10 0 0 0>}}
  }
  TXT_AssetPath {
   widgetClassPath=&TextBlock
   parent=HorizontalBox_2
   widget=TXT_AssetPath
   properties {
    justification=Center
    text="Asset Path:   "
    font {
     bForceMonospaced=false
     bMaterialIsStencil=false
     fontMaterial=null
     fontObject=&Roboto
     letterSpacing=0
     monospacedWidth=1
     size=9
     skewAmount=0
     typefaceFontName=Bold
     outlineSettings {
      bApplyOutlineToDropShadows=false
      bMiteredCorners=false
      bSeparateFillAlpha=false
      outlineMaterial=null
      outlineSize=0
      outlineColor=<rgba 0 0 0 1>
     }
    }
   }
   slot {class=HorizontalBoxSlot, properties {padding=<ltrb 5 5 5 5>}}
  }
  TB_AssetPath {
   widgetClassPath=&EditableTextBox
   bIsVariable=true
   parent=HorizontalBox_2
   widget=TB_AssetPath
   properties {widgetStyle {
     backgroundColor {colorUseRule=UseColor_Specified, specifiedColor=<rgba 1 1 1 1>}
     backgroundImageFocused {
      bIsDynamicallyLoaded=false
      drawAs=RoundedBox
      imageType=NoImage
      mirroring=NoMirror
      resourceName=None
      resourceObject=null
      tiling=NoTile
      imageSize=<0 0>
      margin=<ltrb 0 0 0 0>
      outlineSettings {
       bUseBrushTransparency=false
       roundingType=FixedRadius
       width=1
       color {colorUseRule=UseColor_ColorTable, specifiedColor=<rgba 0 0 0 0>}
       cornerRadii=<4 4 4 4>
      }
      tintColor {colorUseRule=UseColor_Specified, specifiedColor=<rgba 0.0048 0.0048 0.0048 1>}
      uVRegion {bIsValid=false, max=<0 0>, min=<0 0>}
     }
     backgroundImageHovered {
      bIsDynamicallyLoaded=false
      drawAs=RoundedBox
      imageType=NoImage
      mirroring=NoMirror
      resourceName=None
      resourceObject=null
      tiling=NoTile
      imageSize=<0 0>
      margin=<ltrb 0 0 0 0>
      outlineSettings {
       bUseBrushTransparency=false
       roundingType=FixedRadius
       width=1
       color {colorUseRule=UseColor_ColorTable, specifiedColor=<rgba 0 0 0 0>}
       cornerRadii=<4 4 4 4>
      }
      tintColor {colorUseRule=UseColor_Specified, specifiedColor=<rgba 0.0048 0.0048 0.0048 1>}
      uVRegion {bIsValid=false, max=<0 0>, min=<0 0>}
     }
     backgroundImageNormal {
      bIsDynamicallyLoaded=false
      drawAs=RoundedBox
      imageType=NoImage
      mirroring=NoMirror
      resourceName=None
      resourceObject=null
      tiling=NoTile
      imageSize=<0 0>
      margin=<ltrb 0 0 0 0>
      outlineSettings {
       bUseBrushTransparency=false
       roundingType=FixedRadius
       width=1
       color {colorUseRule=UseColor_ColorTable, specifiedColor=<rgba 0 0 0 0>}
       cornerRadii=<4 4 4 4>
      }
      tintColor {colorUseRule=UseColor_Specified, specifiedColor=<rgba 0.0048 0.0048 0.0048 1>}
      uVRegion {bIsValid=false, max=<0 0>, min=<0 0>}
     }
     backgroundImageReadOnly {
      bIsDynamicallyLoaded=false
      drawAs=RoundedBox
      imageType=NoImage
      mirroring=NoMirror
      resourceName=None
      resourceObject=null
      tiling=NoTile
      imageSize=<0 0>
      margin=<ltrb 0 0 0 0>
      outlineSettings {
       bUseBrushTransparency=false
       roundingType=FixedRadius
       width=1
       color {colorUseRule=UseColor_ColorTable, specifiedColor=<rgba 0 0 0 0>}
       cornerRadii=<4 4 4 4>
      }
      tintColor {colorUseRule=UseColor_Specified, specifiedColor=<rgba 0.0284 0.0284 0.0284 1>}
      uVRegion {bIsValid=false, max=<0 0>, min=<0 0>}
     }
     focusedForegroundColor {colorUseRule=UseColor_Specified, specifiedColor=<rgba 1 1 1 1>}
     foregroundColor {colorUseRule=UseColor_Specified, specifiedColor=<rgba 0.5271 0.5271 0.5271 1>}
     hScrollBarPadding=<ltrb 0 0 0 0>
     padding=<ltrb 8 3 8 4>
     readOnlyForegroundColor {colorUseRule=UseColor_Foreground, specifiedColor=<rgba 1 0 1 1>}
     scrollBarStyle {
      thickness=8
      draggedThumbImage {
       bIsDynamicallyLoaded=false
       drawAs=RoundedBox
       imageType=NoImage
       mirroring=NoMirror
       resourceName=None
       resourceObject=null
       tiling=NoTile
       imageSize=<0 0>
       margin=<ltrb 0 0 0 0>
       outlineSettings {
        bUseBrushTransparency=false
        roundingType=FixedRadius
        width=0
        color {colorUseRule=UseColor_Specified, specifiedColor=<rgba 0 0 0 0>}
        cornerRadii=<4 4 4 4>
       }
       tintColor {colorUseRule=UseColor_Specified, specifiedColor=<rgba 0.2159 0.2159 0.2159 1>}
       uVRegion {bIsValid=false, max=<0 0>, min=<0 0>}
      }
      horizontalBackgroundImage {
       bIsDynamicallyLoaded=false
       drawAs=NoDrawType
       imageType=NoImage
       mirroring=NoMirror
       resourceName=None
       resourceObject=null
       tiling=NoTile
       imageSize=<0 0>
       margin=<ltrb 0 0 0 0>
       outlineSettings {
        bUseBrushTransparency=false
        roundingType=HalfHeightRadius
        width=0
        color {colorUseRule=UseColor_Specified, specifiedColor=<rgba 0 0 0 0>}
        cornerRadii=<0 0 0 0>
       }
       tintColor {colorUseRule=UseColor_Specified, specifiedColor=<rgba 1 1 1 1>}
       uVRegion {bIsValid=false, max=<0 0>, min=<0 0>}
      }
      horizontalBottomSlotImage {
       bIsDynamicallyLoaded=false
       drawAs=NoDrawType
       imageType=NoImage
       mirroring=NoMirror
       resourceName=None
       resourceObject=null
       tiling=NoTile
       imageSize=<0 0>
       margin=<ltrb 0 0 0 0>
       outlineSettings {
        bUseBrushTransparency=false
        roundingType=HalfHeightRadius
        width=0
        color {colorUseRule=UseColor_Specified, specifiedColor=<rgba 0 0 0 0>}
        cornerRadii=<0 0 0 0>
       }
       tintColor {colorUseRule=UseColor_Specified, specifiedColor=<rgba 1 1 1 1>}
       uVRegion {bIsValid=false, max=<0 0>, min=<0 0>}
      }
      horizontalTopSlotImage {
       bIsDynamicallyLoaded=false
       drawAs=NoDrawType
       imageType=NoImage
       mirroring=NoMirror
       resourceName=None
       resourceObject=null
       tiling=NoTile
       imageSize=<0 0>
       margin=<ltrb 0 0 0 0>
       outlineSettings {
        bUseBrushTransparency=false
        roundingType=HalfHeightRadius
        width=0
        color {colorUseRule=UseColor_Specified, specifiedColor=<rgba 0 0 0 0>}
        cornerRadii=<0 0 0 0>
       }
       tintColor {colorUseRule=UseColor_Specified, specifiedColor=<rgba 1 1 1 1>}
       uVRegion {bIsValid=false, max=<0 0>, min=<0 0>}
      }
      hoveredThumbImage {
       bIsDynamicallyLoaded=false
       drawAs=RoundedBox
       imageType=NoImage
       mirroring=NoMirror
       resourceName=None
       resourceObject=null
       tiling=NoTile
       imageSize=<0 0>
       margin=<ltrb 0 0 0 0>
       outlineSettings {
        bUseBrushTransparency=false
        roundingType=FixedRadius
        width=0
        color {colorUseRule=UseColor_Specified, specifiedColor=<rgba 0 0 0 0>}
        cornerRadii=<4 4 4 4>
       }
       tintColor {colorUseRule=UseColor_Specified, specifiedColor=<rgba 0.2159 0.2159 0.2159 1>}
       uVRegion {bIsValid=false, max=<0 0>, min=<0 0>}
      }
      normalThumbImage {
       bIsDynamicallyLoaded=false
       drawAs=RoundedBox
       imageType=NoImage
       mirroring=NoMirror
       resourceName=None
       resourceObject=null
       tiling=NoTile
       imageSize=<0 0>
       margin=<ltrb 0 0 0 0>
       outlineSettings {
        bUseBrushTransparency=false
        roundingType=FixedRadius
        width=0
        color {colorUseRule=UseColor_Specified, specifiedColor=<rgba 0 0 0 0>}
        cornerRadii=<4 4 4 4>
       }
       tintColor {colorUseRule=UseColor_Specified, specifiedColor=<rgba 0.0953 0.0953 0.0953 1>}
       uVRegion {bIsValid=false, max=<0 0>, min=<0 0>}
      }
      verticalBackgroundImage {
       bIsDynamicallyLoaded=false
       drawAs=NoDrawType
       imageType=NoImage
       mirroring=NoMirror
       resourceName=None
       resourceObject=null
       tiling=NoTile
       imageSize=<0 0>
       margin=<ltrb 0 0 0 0>
       outlineSettings {
        bUseBrushTransparency=false
        roundingType=HalfHeightRadius
        width=0
        color {colorUseRule=UseColor_Specified, specifiedColor=<rgba 0 0 0 0>}
        cornerRadii=<0 0 0 0>
       }
       tintColor {colorUseRule=UseColor_Specified, specifiedColor=<rgba 1 1 1 1>}
       uVRegion {bIsValid=false, max=<0 0>, min=<0 0>}
      }
      verticalBottomSlotImage {
       bIsDynamicallyLoaded=false
       drawAs=NoDrawType
       imageType=NoImage
       mirroring=NoMirror
       resourceName=None
       resourceObject=null
       tiling=NoTile
       imageSize=<0 0>
       margin=<ltrb 0 0 0 0>
       outlineSettings {
        bUseBrushTransparency=false
        roundingType=HalfHeightRadius
        width=0
        color {colorUseRule=UseColor_Specified, specifiedColor=<rgba 0 0 0 0>}
        cornerRadii=<0 0 0 0>
       }
       tintColor {colorUseRule=UseColor_Specified, specifiedColor=<rgba 1 1 1 1>}
       uVRegion {bIsValid=false, max=<0 0>, min=<0 0>}
      }
      verticalTopSlotImage {
       bIsDynamicallyLoaded=false
       drawAs=NoDrawType
       imageType=NoImage
       mirroring=NoMirror
       resourceName=None
       resourceObject=null
       tiling=NoTile
       imageSize=<0 0>
       margin=<ltrb 0 0 0 0>
       outlineSettings {
        bUseBrushTransparency=false
        roundingType=HalfHeightRadius
        width=0
        color {colorUseRule=UseColor_Specified, specifiedColor=<rgba 0 0 0 0>}
        cornerRadii=<0 0 0 0>
       }
       tintColor {colorUseRule=UseColor_Specified, specifiedColor=<rgba 1 1 1 1>}
       uVRegion {bIsValid=false, max=<0 0>, min=<0 0>}
      }
     }
     textStyle {
      overflowPolicy=Clip
      transformPolicy=None
      colorAndOpacity {colorUseRule=UseColor_Foreground, specifiedColor=<rgba 1 0 1 1>}
      font {
       bForceMonospaced=false
       bMaterialIsStencil=false
       fontMaterial=null
       fontObject=null
       letterSpacing=0
       monospacedWidth=1
       size=10
       skewAmount=0
       typefaceFontName=Regular
       outlineSettings {
        bApplyOutlineToDropShadows=false
        bMiteredCorners=false
        bSeparateFillAlpha=false
        outlineMaterial=null
        outlineSize=0
        outlineColor=<rgba 0 0 0 1>
       }
      }
      highlightColor {colorUseRule=UseColor_ColorTable, specifiedColor=<rgba 0 0 0 0>}
      highlightShape {
       bIsDynamicallyLoaded=false
       drawAs=Image
       imageType=NoImage
       mirroring=NoMirror
       resourceName=None
       resourceObject=null
       tiling=NoTile
       imageSize=<0 0>
       margin=<ltrb 0 0 0 0>
       outlineSettings {
        bUseBrushTransparency=false
        roundingType=HalfHeightRadius
        width=0
        color {colorUseRule=UseColor_Specified, specifiedColor=<rgba 0 0 0 0>}
        cornerRadii=<0 0 0 0>
       }
       tintColor {colorUseRule=UseColor_ColorTable, specifiedColor=<rgba 0 0 0 0>}
       uVRegion {bIsValid=false, max=<0 0>, min=<0 0>}
      }
      selectedBackgroundColor {colorUseRule=UseColor_ColorTable, specifiedColor=<rgba 0 0 0 0>}
      shadowColorAndOpacity=<rgba 0 0 0 1>
      shadowOffset=<0 0>
      strikeBrush {
       bIsDynamicallyLoaded=false
       drawAs=Image
       imageType=NoImage
       mirroring=NoMirror
       resourceName=None
       resourceObject=null
       tiling=NoTile
       imageSize=<32 32>
       margin=<ltrb 0 0 0 0>
       outlineSettings {
        bUseBrushTransparency=false
        roundingType=HalfHeightRadius
        width=0
        color {colorUseRule=UseColor_Specified, specifiedColor=<rgba 0 0 0 0>}
        cornerRadii=<0 0 0 0>
       }
       tintColor {colorUseRule=UseColor_Specified, specifiedColor=<rgba 1 1 1 1>}
       uVRegion {bIsValid=false, max=<0 0>, min=<0 0>}
      }
      underlineBrush {
       bIsDynamicallyLoaded=false
       drawAs=Image
       imageType=NoImage
       mirroring=NoMirror
       resourceName=None
       resourceObject=null
       tiling=NoTile
       imageSize=<32 32>
       margin=<ltrb 0 0 0 0>
       outlineSettings {
        bUseBrushTransparency=false
        roundingType=HalfHeightRadius
        width=0
        color {colorUseRule=UseColor_Specified, specifiedColor=<rgba 0 0 0 0>}
        cornerRadii=<0 0 0 0>
       }
       tintColor {colorUseRule=UseColor_Specified, specifiedColor=<rgba 1 1 1 1>}
       uVRegion {bIsValid=false, max=<0 0>, min=<0 0>}
      }
     }
     vScrollBarPadding=<ltrb 0 0 0 0>
    }}
   slot {class=HorizontalBoxSlot, properties {size {sizeRule=Fill, value=1}}}
  }
  DV_CustomOptions {
   widgetClassPath=&DetailsView
   parent=VerticalBox_0
   widget=DV_CustomOptions
   properties {categoriesToShow [CustomOptions]}
   slot {class=VerticalBoxSlot}
  }
  HorizontalBox_4 {widgetClassPath=&HorizontalBox, parent=VerticalBox_0, widget=HorizontalBox_4, slot {class=VerticalBoxSlot}}
  BTN_CreateNewAsset {
   widgetClassPath=&TA_Generic01_Button
   parent=HorizontalBox_4
   widget=BTN_CreateNewAsset
   properties {label=Create New Asset}
   slot {class=HorizontalBoxSlot, properties {horizontalAlignment=HAlign_Center, padding=<ltrb 5 5 5 5>}}
  }
 ]
}
